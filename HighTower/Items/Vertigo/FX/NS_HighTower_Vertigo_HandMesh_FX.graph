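# Niagara system for the HighTower Vertigo item's hand mesh effect, with all three of its emitters disabled as saved.
# Drops_02 spawns short-lived sprite droplets that inherit velocity, fall under gravity and drag, and stretch with speed.
# Tendrils spawns dark, randomized meshes, driving material bend, width, panning and wobble through dynamic parameters and curves.
# Tendrils_New bursts a single rotating mesh with curve-driven material parameters, alpha fade and decaying camera offset.
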NiagaraSystem /HighTower/Items/Vertigo/FX/NS_HighTower_Vertigo_HandMesh_FX
dependencies {
 usedDataInterfaces [&NiagaraDataInterfaceCurve]
 usedDynamicInputs []
 usedModules [/Niagara/Modules/System/SystemLifeCycle, &SystemState]
}
emitters {
 Drops_02 {
  bEnabled=false
  emitterSpawnScript [null() {enabled=false, module=PreserveSkydiveUserVariables}]
  emitterUpdateScript [
   &EmitterState(Life Cycle Mode=::NewEnumerator1, Loop Duration=1, UseLoopDelay=false)
   &SpawnRate(SpawnRate=10, Use Spawn Probability=false)
  ]
  particleSpawnScript [
   /Niagara/Modules/Spawn/Initialization/InitializeParticle(
    Lifetime=&UniformRangedFloat(Minimum=0.4, Maximum=0.6)
    Module.Write Color=true
    Module.Write Lifetime=true
    Module.Write Mass=false
    Module.Write Material Random=true
    Module.Write Position=false
    Module.Write Scale=false
    Module.Write SpriteRotation=false
    Module.Write SpriteSize=true
    Position=&SimulationPosition()
    Sprite Rotation=&UniformRangedFloat(Minimum=0, Maximum=1)
    Sprite Size=&Vector2DFromFloat(Value=&UniformRangedFloat(Minimum=1.5, Maximum=3))
   )
   /Niagara/Modules/Update/SubUV/SubUVAnimation(SubUV Animation Mode=::NewEnumerator1)
   &AddVelocity(Velocity=&Multiply_Vector(
     A=&VectorFromFloat(Value=200)
     B=&UniformRangedVector(Minimum=<-1 -1 -1>, Maximum=<1 1 1>)
    ))
   &InheritVelocity(
    Module.Limit Speed=true
    Velocity Limit=500
    Velocity Scale=&Multiply_Vector(
     A=<1 1 1>
     B=&VectorFromFloat(Value=&UniformRangedFloat(Minimum=1, Maximum=5))
    )
   )
  ]
  particleUpdateScript [
   &ParticleState()
   &Color(Scale Alpha=&FloatFromCurve(FloatCurve=DataInterface() {propertyValues {Curve {keys [<at -0.0046 1.0113>,<at 1.0011 0.002>]}}}))
   &ScaleSpriteSize(Scale Factor=&Vector2DFromFloat(Value=&FloatFromCurve(FloatCurve=DataInterface() {propertyValues {Curve {keys [<at 0 0>,<at 0.0524 1>,<at 0.85 1>,<at 1.0011 0.002>]}}})))
   &Drag(Drag=2, Use Linear Drag=true, Use Rotational Drag=false)
   &GravityForce()
   &SolveForcesAndVelocity(Clamp Velocity=false, Limit Acceleration=false)
   &SpriteRotationRate(Rotation Rate=$Particles.InitRotationRate)
   &ScaleSpriteSizeBySpeed(Max Scale Factor=<1 20>)
  ]
  rendererClasses [&NiagaraSpriteRendererProperties]
  renderers [{rendererClass=&NiagaraSpriteRendererProperties}]
 }
 Tendrils {
  bEnabled=false
  emitterUpdateScript [
   &EmitterState(Life Cycle Mode=::NewEnumerator1, Loop Duration=1, UseLoopDelay=false)
   &SpawnRate(SpawnRate=5, Use Spawn Probability=false)
   &SpawnBurst_Instantaneous(Spawn Count=6, Use Spawn Probability=false)
  ]
  particleSpawnScript [
   /SetVariables(
    Particles.BendDirectionX:NiagaraFloat=&UniformRangedFloat(Minimum=0, Maximum=1)
    Particles.BendDirectionY:NiagaraFloat=&UniformRangedFloat(Minimum=0, Maximum=1)
    Particles.BendDirectionZ:NiagaraFloat=&UniformRangedFloat(Minimum=0, Maximum=1)
    Particles.BendStrength:NiagaraFloat=&UniformRangedFloat(Minimum=60, Maximum=150)
    Particles.CurveOffset:NiagaraFloat=&UniformRangedFloat(Minimum=0, Maximum=1)
    Particles.PanSpeed:NiagaraFloat=&UniformRangedFloat(Minimum=-1, Maximum=-2)
    Particles.SinScale:NiagaraFloat=&UniformRangedFloat(Minimum=0.2, Maximum=0.4)
    Particles.SinSpeed:NiagaraFloat=&Lerp_Float(A=-8, B=-15, Alpha=$Particles.SinScale)
    Particles.WidthRandom:NiagaraFloat=&UniformRangedFloat(Minimum=0.3, Maximum=1.7)
   ) {isSetParameters=true}
   &InitializeParticle(
    Color=<rgba 0.0052 0.0052 0.0052 1>
    Lifetime=&UniformRangedFloat(Minimum=0.5, Maximum=0.7)
    Mesh Scale=&Multiply_Vector(
     A=&UniformRangedVector(Minimum=<1 0.5 0.5>, Maximum=<1 0.8 0.8>)
     B=&VectorFromFloat(Value=&UniformRangedFloat(Minimum=0.7, Maximum=1))
    )
    Module.Write Color=true
    Module.Write Lifetime=true
    Module.Write Mass=true
    Module.Write Material Random=true
    Module.Write Position=false
    Module.Write Scale=true
    Module.Write SpriteRotation=false
    Module.Write SpriteSize=true
    Position=&SimulationPosition()
    Sprite Size=<5 5>
   )
   /Niagara/Modules/Spawn/Orientation/V2/InitialMeshOrientation(
    Rotation=&UniformRangedVector(Minimum=<-1 -1 -1>, Maximum=<1 1 1>)
    Use Orientation Vector=true
    Use Rotation Vector=true
   )
   &AddVelocityFromPoint(Module.Use Falloff Distance=false, Velocity Strength=1)
  ]
  particleUpdateScript [
   &ParticleState()
   &SolveForcesAndVelocity(Clamp Velocity=false, Limit Acceleration=false)
   &DynamicMaterialParameters(
    Index 0 Param 1=&FloatFromCurve(
     FloatCurve=DataInterface() {propertyValues {Curve {keys [
         <at 0 0.2 RCIM_Cubic>
         {arriveTangent=2.0368, interpMode=RCIM_Cubic, leaveTangent=2.0368, time=0.0982, value=2}
         {arriveTangent=-0.6576, interpMode=RCIM_Cubic, leaveTangent=-0.6576, time=0.2706, value=0.7511}
         <at 0.4018 1.8004 RCIM_Cubic>
         {arriveTangent=-1.2864, interpMode=RCIM_Cubic, leaveTangent=-1.2864, time=0.5508, value=0.7511}
         <at 0.7243 1.3855>
         <at 1 0 RCIM_Cubic>
        ]}}}
     Scale Curve=$Particles.WidthRandom
    )
    Index 0 Param 2=$Particles.WidthRandom
    Index 0 Param 3=&FloatFromCurve(FloatCurve=DataInterface() {propertyValues {Curve {keys [
         <at 0 -0.5 RCIM_Cubic>
         {arriveTangent=2.0829, interpMode=RCIM_Cubic, leaveTangent=2.0829, time=0.3991, value=0.9166}
         <at 0.5434 1.0587>
         {arriveTangent=-2.3574, interpMode=RCIM_Cubic, leaveTangent=-2.3574, time=0.6718, value=0.8994}
         <at 1 -0.5 RCIM_Cubic>
        ]}}})
    Index 0 Param 4=$Particles.PanSpeed
    Index 1 Param 1=&FloatFromCurve(
     FloatCurve=DataInterface() {propertyValues {Curve {
        postInfinityExtrap=RCCE_CycleWithOffset
        preInfinityExtrap=RCCE_CycleWithOffset
        keys [
         <at -0.2078 -0.2949 RCIM_Cubic>
         {arriveTangent=-0.1876, interpMode=RCIM_Cubic, leaveTangent=-0.1876, time=-0.1305, value=0.3531}
         {arriveTangent=0.0791, interpMode=RCIM_Cubic, leaveTangent=0.0791, time=-0.0639, value=-0.3219}
         {arriveTangent=-0.0573, interpMode=RCIM_Cubic, leaveTangent=-0.0573, time=0.0458, value=0.3671}
         {arriveTangent=0.1517, interpMode=RCIM_Cubic, leaveTangent=0.1517, time=0.1269, value=-0.3329}
         {arriveTangent=-0.1648, interpMode=RCIM_Cubic, leaveTangent=-0.1648, time=0.2208, value=0.3937}
         {arriveTangent=0.1554, interpMode=RCIM_Cubic, leaveTangent=0.1554, time=0.3191, value=-0.3646}
         {arriveTangent=0.1069, interpMode=RCIM_Cubic, leaveTangent=0.1069, time=0.4235, value=0.4252}
         {arriveTangent=-0.3187, interpMode=RCIM_Cubic, leaveTangent=-0.3187, time=0.5167, value=-0.3434}
         {arriveTangent=0.1258, interpMode=RCIM_Cubic, leaveTangent=0.1258, time=0.6494, value=0.3531}
         {arriveTangent=0.1369, interpMode=RCIM_Cubic, leaveTangent=0.1369, time=0.7949, value=-0.3084}
         {arriveTangent=-0.1066, interpMode=RCIM_Cubic, leaveTangent=-0.1066, time=0.9125, value=0.3892}
         <at 1.0905 -0.3399 RCIM_Cubic>
         {arriveTangent=0.0092, interpMode=RCIM_Cubic, leaveTangent=0.0092, time=1.2112, value=0.3892}
         {arriveTangent=0.0148, interpMode=RCIM_Cubic, leaveTangent=0.0148, time=1.3361, value=-0.3377}
         {arriveTangent=0.0109, interpMode=RCIM_Cubic, leaveTangent=0.0109, time=1.4646, value=0.3929}
         {arriveTangent=-0.8851, interpMode=RCIM_Cubic, leaveTangent=-0.8851, time=1.603, value=-0.3348}
         {arriveTangent=0.5332, interpMode=RCIM_Cubic, leaveTangent=0.5332, time=1.6678, value=0.2131}
         {arriveTangent=0.6123, interpMode=RCIM_Cubic, leaveTangent=0.6123, time=1.7796, value=-0.2406}
         {arriveTangent=-0.2327, interpMode=RCIM_Cubic, leaveTangent=-0.2327, time=1.8915, value=0.3501}
         <at 2.0004 -0.292 RCIM_Cubic>
        ]
       }}}
     CurveIndex=&Add_Float(A=$Particles.NormalizedAge, B=$Particles.CurveOffset)
     Scale Curve=$Particles.BendStrength
    )
    Index 1 Param 2=$Particles.BendDirectionX
    Index 1 Param 3=$Particles.BendDirectionY
    Index 1 Param 4=$Particles.BendDirectionZ
    Index 2 Param 1=$Particles.SinScale
    Index 2 Param 2=$Particles.SinSpeed
    Index 2 Param 3=10
    Module.Param0WriteEnabled=true
    Module.Param0WriteEnabled1=true
    Module.Param0WriteEnabled2=true
    Module.Param1WriteEnabled=true
    Module.Param1WriteEnabled1=true
    Module.Param1WriteEnabled2=true
    Module.Param2WriteEnabled=true
    Module.Param2WriteEnabled1=true
    Module.Param2WriteEnabled2=true
    Module.Param3WriteEnabled=true
    Module.Param3WriteEnabled1=true
    Module.Param3WriteEnabled2=true
    Write Parameter Index 1=true
    Write Parameter Index 2=true
   )
   &ScaleMeshSize(Scale Factor=&VectorFromFloat(Value=&FloatFromCurve(FloatCurve=DataInterface() {propertyValues {Curve {keys [<at 0.5214 1 RCIM_Cubic>,<at 1 0 RCIM_Cubic>]}}})))
  ]
  rendererClasses [&NiagaraSpriteRendererProperties, &NiagaraMeshRendererProperties]
  renderers [
   {rendererClass=&NiagaraSpriteRendererProperties}
   {rendererClass=&NiagaraMeshRendererProperties, rendererIndex=1}
  ]
 }
 Tendrils_New {
  bEnabled=false
  emitterUpdateScript [
   &EmitterState(Life Cycle Mode=::NewEnumerator1, Loop Duration=0.5, UseLoopDelay=false)
   &SpawnRate(Use Spawn Probability=false) {enabled=false}
   &SpawnBurst_Instantaneous(Spawn Count=1, Use Spawn Probability=false)
  ]
  particleSpawnScript [
   &InitializeParticle(
    Color=<rgba 1 1 1 10>
    Lifetime=0.8
    Mesh Scale=&Multiply_Vector(
     A=&UniformRangedVector(Minimum=<0.8 0.5 0.5>, Maximum=<1 1 1>)
     B=&VectorFromFloat(Value=3)
    )
    Module.Write Color=true
    Module.Write Lifetime=true
    Module.Write Mass=true
    Module.Write Material Random=true
    Module.Write Position=false
    Module.Write Scale=true
    Module.Write SpriteRotation=false
    Module.Write SpriteSize=true
    Position=&SimulationPosition()
    Sprite Size=<5 5>
   )
   &InitialMeshOrientation(Rotation=<0 0 0.5>, Use Orientation Vector=true, Use Rotation Vector=true)
   &SystemLocation(MaskSpawnGroup=false)
  ]
  particleUpdateScript [
   &ParticleState()
   &SolveForcesAndVelocity(Clamp Velocity=false, Limit Acceleration=false)
   &DynamicMaterialParameters(
    Index 0 Param 1=&FloatFromCurve(FloatCurve=DataInterface() {propertyValues {Curve {keys [
         <at 0 -2 RCIM_Cubic>
         {arriveTangent=-0.2393, interpMode=RCIM_Cubic, leaveTangent=-0.2393, tangentMode=RCTM_User, time=0.1, value=0}
         {arriveTangent=0.0198, interpMode=RCIM_Cubic, leaveTangent=0.0198, tangentMode=RCTM_User, time=0.6471, value=0}
         {interpMode=RCIM_Cubic, leaveTangent=-21.8502, time=1, value=-3}
        ]}}})
    Index 0 Param 2=1.25
    Index 0 Param 3=2
    Module.Param0WriteEnabled=false
    Module.Param1WriteEnabled=true
    Module.Param2WriteEnabled=true
    Module.Param3WriteEnabled=true
   )
   &CameraOffset(Camera Offset Amount=5) {enabled=false}
   &ScaleColor(
    Scale Alpha=&FloatFromCurve(FloatCurve=DataInterface() {propertyValues {Curve {keys [
         <at 0 0 RCIM_Cubic>
         {arriveTangent=0.1008, interpMode=RCIM_Cubic, leaveTangent=0.1008, tangentMode=RCTM_User, time=0.0523, value=1}
         {arriveTangent=0.275, interpMode=RCIM_Cubic, leaveTangent=0.275, tangentMode=RCTM_User, time=0.6836, value=1}
         <at 1 0 RCIM_Cubic>
        ]}}})
    ScaleA=true
    ScaleRGB=true
   )
   &ScaleMeshSize(Scale Factor=&MakeVector(
     X=&FloatFromCurve(FloatCurve=DataInterface() {propertyValues {Curve {keys [<at 0.0039 1.5 RCIM_Cubic>,<at 0.1737 2>,<at 1 1 RCIM_Cubic>]}}})
     Y=1
     Z=1
    )) {enabled=false}
   &CameraOffset(Camera Offset Amount=&FloatFromCurve(FloatCurve=DataInterface() {propertyValues {Curve {keys [<at 0 1 RCIM_Cubic>,<at 1 0>]}}}, Scale Curve=20)) {module=CameraOffset001}
   /Niagara/Modules/Update/Orientation/V2/UpdateMeshOrientation(Rotation Rate=-0.3, Rotation Vector:Vector3f=<1 0 0>)
  ]
  rendererClasses [&NiagaraSpriteRendererProperties, &NiagaraMeshRendererProperties]
  renderers [
   {rendererClass=&NiagaraSpriteRendererProperties}
   {rendererClass=&NiagaraMeshRendererProperties, rendererIndex=1}
  ]
 }
}
summary {
 systemName=NS_HighTower_Vertigo_HandMesh_FX
 emitters [
  {bEnabled=false, emitterName=Drops_02, rendererClasses [&NiagaraSpriteRendererProperties]}
  {
   bEnabled=false
   emitterName=Tendrils
   rendererClasses [&NiagaraSpriteRendererProperties, &NiagaraMeshRendererProperties]
  }
  {
   bEnabled=false
   emitterName=Tendrils_New
   rendererClasses [&NiagaraSpriteRendererProperties, &NiagaraMeshRendererProperties]
  }
 ]
 userVariables [
  User.bIsFrontEndPreview {type=NiagaraBool}
  User.FEVec {type=Vector3f, description="", defaultValue {struct=Vector3f, value=<0 0 0>}}
  User.PlayerSpeed {type=NiagaraFloat, defaultValue=0, description=""}
  User.PreviousVelocityXY {type=Vector3f, description="", defaultValue {struct=Vector3f, value=<0 0 0>}}
  User.RibbonIDOffset {type=NiagaraInt32}
  User.TargetPosition {type=Vector3f, description="", defaultValue {struct=Vector3f, value=<0 0 0>}}
  User.WaterBodyIndex {type=NiagaraFloat, defaultValue=0, description=""}
  User.WaterDepth {type=NiagaraFloat, defaultValue=0, description=""}
  User.WaterPlaneLocation {type=Vector3f, description="", defaultValue {struct=Vector3f, value=<0 0 0>}}
  User.WaterPlaneNormal {type=Vector3f, description="", defaultValue {struct=Vector3f, value=<0 0 0>}}
  User.WaveTime {type=NiagaraFloat, defaultValue=0, description=""}
 ]
}
