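# Niagara effect for the HighTower Vertigo item's hand, with dark tendril, liquid-mesh and drop emitters, all disabled in the portion shown.
# Arm emitters scatter particles between owner and User.TargetPosition.
# ArmGuid weaves toward that target using curve-driven perpendicular offsets, emitting location events, as does Hand_Around_Lead, alongside companion ribbon emitters.
# SimpleSpriteBurst hangs pendulum-constrained meshes from User.SkeletalMesh, passing pivot and position to dynamic material parameters.
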NiagaraSystem /HighTower/Items/Vertigo/FX/NS_HighTower_Vertigo_Hand_FX
dependencies {
 usedDataInterfaces [&NiagaraDataInterfaceCurve]
 usedDynamicInputs []
 usedModules [/Niagara/Modules/System/SystemLifeCycle, &SystemState]
}
emitters {
 ArmGuid {
  bEnabled=false
  emitterUpdateScript [
   &EmitterState(Life Cycle Mode=::NewEnumerator1, Loop Behavior=::NewEnumerator1, Loop Duration=1, UseLoopDelay=false)
   &SpawnRate(SpawnRate=20, Use Spawn Probability=false) {enabled=false}
   &SpawnBurst_Instantaneous(Spawn Count=3, Use Spawn Probability=false)
  ]
  particleSpawnScript [
   /SetVariables(
    Particles.AgeIndexRandom:NiagaraFloat=&UniformRangedFloat(Minimum=0, Maximum=0.2)
    Particles.SinIntensity:NiagaraFloat=&UniformRangedFloat(Minimum=20, Maximum=30)
    Particles.VectorRandom:Vector3f=/Niagara/DynamicInputs/Vectors/CrossProduct(
     Vector A:Vector3f=<0 0 1>
     Vector B:Vector3f=&NormalizeVector(Vector To Normalize=&Subtract_Vector(A=$User.TargetPosition, B=Linked() {linkedVariable {name=Engine.Owner.Position, type=NiagaraPosition}}))
     Normalize Incoming Vectors:NiagaraBool=true
    )
   ) {isSetParameters=true}
   /Niagara/Modules/Spawn/Initialization/InitializeParticle(
    Lifetime=&UniformRangedFloat(Minimum=0.2, Maximum=1)
    Module.Write Color=true
    Module.Write Lifetime=true
    Module.Write Mass=true
    Module.Write Material Random=true
    Module.Write Position=false
    Module.Write Scale=false
    Module.Write SpriteRotation=false
    Module.Write SpriteSize=true
    Position=&SimulationPosition()
    Sprite Size=<20 20>
   )
  ]
  particleUpdateScript [
   &ParticleState()
   &SystemLocation(
    MaskSpawnGroup=false
    Offset=&Add_Vector(
     A=&Multiply_Vector(
      A=&Subtract_Vector(A=$User.TargetPosition, B=Linked() {linkedVariable {name=Engine.Owner.Position, type=NiagaraPosition}})
      B=&VectorFromFloat(Value=0)
     )
     B=&Add_Vector(
      A=&Add_Vector(
       A=&Multiply_Vector(
        A=&Multiply_Vector(A=$Particles.VectorRandom, B=&VectorFromFloat(Value=$Particles.SinIntensity))
        B=&VectorFromFloat(Value=&FloatFromCurve(
          FloatCurve=DataInterface() {propertyValues {Curve {keys [
              <at 0.0242 0 RCIM_Cubic>
              {arriveTangent=6.3746, interpMode=RCIM_Cubic, leaveTangent=6.3746, time=0.1094, value=0.2297}
              {arriveTangent=-5.3965, interpMode=RCIM_Cubic, leaveTangent=-5.3965, time=0.2368, value=1.3548}
              {arriveTangent=-0.0558, interpMode=RCIM_Cubic, leaveTangent=-0.0558, time=0.3975, value=-1.3247}
              {arriveTangent=0.6978, interpMode=RCIM_Cubic, leaveTangent=0.6978, time=0.5626, value=1.3366}
              {arriveTangent=0.5192, interpMode=RCIM_Cubic, leaveTangent=0.5192, time=0.7388, value=-1.0866}
              {arriveTangent=2.8925, interpMode=RCIM_Cubic, leaveTangent=2.8925, time=0.9034, value=1.5136}
              <at 1.1145 0 RCIM_Cubic>
             ]}}}
          CurveIndex=&Add_Float(A=$Particles.NormalizedAge, B=$Particles.AgeIndexRandom)
         ))
       )
       B=&Multiply_Vector(
        A=&Multiply_Vector(
         A=&CrossProduct(
          Vector A:Vector3f=&NormalizeVector(Vector To Normalize=&Subtract_Vector(A=$User.TargetPosition, B=Linked() {linkedVariable {name=Engine.Owner.Position, type=NiagaraPosition}}))
          Vector B:Vector3f=$Particles.VectorRandom
          Normalize Incoming Vectors:NiagaraBool=true
         )
         B=&VectorFromFloat(Value=$Particles.SinIntensity)
        )
        B=&VectorFromFloat(Value=&FloatFromCurve(
          FloatCurve=DataInterface() {propertyValues {Curve {keys [
              <at 0.0242 0 RCIM_Cubic>
              {arriveTangent=-5.3558, interpMode=RCIM_Cubic, leaveTangent=-5.3558, time=0.1216, value=1.2695}
              <at 0.2846 -1.3943 RCIM_Cubic>
              {arriveTangent=0.6222, interpMode=RCIM_Cubic, leaveTangent=0.6222, time=0.4582, value=1.2695}
              {arriveTangent=0.7737, interpMode=RCIM_Cubic, leaveTangent=0.7737, time=0.6083, value=-1.1929}
              {arriveTangent=3.0457, interpMode=RCIM_Cubic, leaveTangent=3.0457, time=0.8054, value=1.5381}
              <at 1 0 RCIM_Cubic>
             ]}}}
          CurveIndex=&Add_Float(A=$Particles.NormalizedAge, B=$Particles.AgeIndexRandom)
         ))
       )
      )
      B=&Multiply_Vector(
       A=&Subtract_Vector(A=$User.TargetPosition, B=Linked() {linkedVariable {name=Engine.Owner.Position, type=NiagaraPosition}})
       B=&VectorFromFloat(Value=&FloatFromCurve(FloatCurve=DataInterface() {propertyValues {Curve {keys [
             <at 0 0 RCIM_Cubic>
             {arriveTangent=1.2008, interpMode=RCIM_Cubic, leaveTangent=1.2008, tangentMode=RCTM_User, time=0.4234, value=0.1389}
             <at 1 1 RCIM_Cubic>
            ]}}}))
      )
     )
    )
    Offset Coordinate Space=::World
   )
   &SolveForcesAndVelocity(Clamp Velocity=false, Limit Acceleration=false)
   &GenerateLocationEvent(Module.Use Event Send Rate=true)
  ]
  rendererClasses [&NiagaraSpriteRendererProperties]
  renderers [{rendererClass=&NiagaraSpriteRendererProperties}]
 }
 ArmStrike {
  bEnabled=false
  emitterUpdateScript [
   &EmitterState(Life Cycle Mode=::NewEnumerator1, Loop Duration=1, UseLoopDelay=false)
   &SpawnRate(SpawnRate=0, Use Spawn Probability=false)
  ]
  particleSpawnScript [
   &InitializeParticle(
    Module.Write Color=true
    Module.Write Lifetime=true
    Module.Write Mass=true
    Module.Write Material Random=true
    Module.Write Position=false
    Module.Write Scale=false
    Module.Write SpriteRotation=false
    Module.Write SpriteSize=true
    Position=&SimulationPosition()
    Sprite Size=<5 5>
   )
   &InitializeRibbon(
    Lifetime=0.3
    Module.Write Color=true
    Module.Write Lifetime=true
    Module.Write Mass=true
    Module.Write Material Random=true
    Module.Write Position=false
    Module.Write RibbonTwist=false
    Module.Write RibbonWidth=true
    Ribbon Width=10
   )
  ]
  particleUpdateScript [
   &ParticleState()
   &SolveForcesAndVelocity(Clamp Velocity=false, Limit Acceleration=false)
   &ScaleRibbonWidth()
  ]
  rendererClasses [&NiagaraSpriteRendererProperties, &NiagaraRibbonRendererProperties]
  renderers [
   {rendererClass=&NiagaraSpriteRendererProperties}
   {rendererClass=&NiagaraRibbonRendererProperties, rendererIndex=1}
  ]
 }
 Arm_Net {
  bEnabled=false
  emitterUpdateScript [
   &EmitterState(Life Cycle Mode=::NewEnumerator1, Loop Behavior=::NewEnumerator1, Loop Duration=1, UseLoopDelay=false)
   &SpawnRate(SpawnRate=20, Use Spawn Probability=false)
  ]
  particleSpawnScript [
   /SetVariables(
    Particles.PositionRandom:NiagaraFloat=&UniformRangedFloat(Minimum=0, Maximum=1)
    Particles.SizeRandom:NiagaraFloat=&UniformRangedFloat(Minimum=0.13, Maximum=0.1)
    Particles.SizeRandomXZ:NiagaraFloat=&UniformRangedFloat(Minimum=0.2, Maximum=0.4)
   ) {isSetParameters=true}
   /SetVariables(
    Particles.D_TimeOffset:NiagaraFloat=&UniformRangedFloat(Minimum=0, Maximum=5)
    Particles.UV_Scale:NiagaraFloat=&UniformRangedFloat(Minimum=0.1, Maximum=0.5)
   ) {isSetParameters=true, module=SetVariables_#1}
   &InitializeParticle(
    Color=<rgba 0.0053 0.0068 0.0156 1>
    Lifetime=&UniformRangedFloat(Minimum=0.5, Maximum=0.8)
    Mesh Scale=&Multiply_VectorByFloat(
     Vector=&UniformRangedVector(Minimum=<0.5 1 2>, Maximum=<1 2 1.5>)
     Float=&UniformRangedFloat(Minimum=0.4, Maximum=0.2)
    )
    Module.Write Color=true
    Module.Write Lifetime=true
    Module.Write Mass=true
    Module.Write Material Random=true
    Module.Write Position=false
    Module.Write Scale=true
    Module.Write SpriteRotation=true
    Module.Write SpriteSize=true
    Position=&SimulationPosition()
    Sprite Rotation=&UniformRangedFloat(Minimum=0, Maximum=1)
    Sprite Size=&Multiply_Vector2(
     A=<50 50>
     B=&Vector2DFromFloat(Value=&UniformRangedFloat(Minimum=0.25, Maximum=0.6))
    )
   )
   &SystemLocation(
    MaskSpawnGroup=false
    Offset=&Multiply_Vector(
     A=&Subtract_Vector(A=$User.TargetPosition, B=Linked() {linkedVariable {name=Engine.Owner.Position, type=NiagaraPosition}})
     B=&VectorFromFloat(Value=$Particles.PositionRandom)
    )
    Offset Coordinate Space=::World
   )
   &AddVelocity(Velocity=&Multiply_Vector(
     A=&UniformRangedVector(Minimum=<-1 -1 -1>, Maximum=<1 1 1>)
     B=&VectorFromFloat(Value=100)
    ))
  ]
  particleUpdateScript [
   &ParticleState()
   &DynamicMaterialParameters(
    Index 0 Param 1=$Particles.D_TimeOffset
    Index 0 Param 2=$Particles.UV_Scale
    Index 0 Param 3=0.6
    Index 0 Param 4=0.4
    Module.Param0WriteEnabled=true
    Module.Param1WriteEnabled=true
    Module.Param2WriteEnabled=true
    Module.Param3WriteEnabled=true
   )
   &ScaleColor(
    Scale Alpha=&FloatFromCurve(FloatCurve=DataInterface() {propertyValues {Curve {keys [<at 0 0 RCIM_Cubic>,<at 0.502 2.9766 RCIM_Cubic>,<at 1 0>]}}})
    ScaleA=true
    ScaleRGB=true
   )
   &Drag(Drag=5, Use Linear Drag=true, Use Rotational Drag=false)
   &SolveForcesAndVelocity(Clamp Velocity=false, Limit Acceleration=false)
  ]
  rendererClasses [&NiagaraMeshRendererProperties, &NiagaraSpriteRendererProperties]
  renderers [
   {rendererClass=&NiagaraMeshRendererProperties, rendererIndex=0}
   {rendererClass=&NiagaraSpriteRendererProperties, rendererIndex=1}
  ]
 }
 Drops_02 {
  bEnabled=false
  emitterSpawnScript [null() {enabled=false, module=PreserveSkydiveUserVariables}]
  emitterUpdateScript [
   &EmitterState(Life Cycle Mode=::NewEnumerator1, Loop Behavior=::NewEnumerator1, Loop Duration=1.5, UseLoopDelay=false)
   &SpawnRate(SpawnRate=10, Use Spawn Probability=false)
  ]
  particleSpawnScript [
   &InitializeParticle(
    Lifetime=&UniformRangedFloat(Minimum=0.4, Maximum=0.6)
    Module.Write Color=true
    Module.Write Lifetime=true
    Module.Write Mass=false
    Module.Write Material Random=true
    Module.Write Position=false
    Module.Write Scale=false
    Module.Write SpriteRotation=false
    Module.Write SpriteSize=true
    Position=&SimulationPosition()
    Sprite Rotation=&UniformRangedFloat(Minimum=0, Maximum=1)
    Sprite Size=&Vector2DFromFloat(Value=&UniformRangedFloat(Minimum=1.5, Maximum=3))
   )
   /Niagara/Modules/Update/SubUV/SubUVAnimation(SubUV Animation Mode=::NewEnumerator1)
   &AddVelocity(Velocity=&Multiply_Vector(
     A=&VectorFromFloat(Value=200)
     B=&UniformRangedVector(Minimum=<-1 -1 -1>, Maximum=<1 1 1>)
    ))
   &InheritVelocity(
    Module.Limit Speed=true
    Velocity Limit=500
    Velocity Scale=&Multiply_Vector(
     A=<1 1 1>
     B=&VectorFromFloat(Value=&UniformRangedFloat(Minimum=1, Maximum=5))
    )
   )
  ]
  particleUpdateScript [
   &ParticleState()
   &Color(Scale Alpha=&FloatFromCurve(FloatCurve=DataInterface() {propertyValues {Curve {keys [<at -0.0046 1.0113>,<at 1.0011 0.002>]}}}))
   &ScaleSpriteSize(Scale Factor=&Vector2DFromFloat(Value=&FloatFromCurve(FloatCurve=DataInterface() {propertyValues {Curve {keys [<at 0 0>,<at 0.0524 1>,<at 0.85 1>,<at 1.0011 0.002>]}}})))
   &Drag(Drag=2, Use Linear Drag=true, Use Rotational Drag=false)
   &GravityForce()
   &SolveForcesAndVelocity(Clamp Velocity=false, Limit Acceleration=false)
   &SpriteRotationRate(Rotation Rate=$Particles.InitRotationRate)
   &ScaleSpriteSizeBySpeed(Max Scale Factor=<1 20>)
  ]
  rendererClasses [&NiagaraSpriteRendererProperties]
  renderers [{rendererClass=&NiagaraSpriteRendererProperties}]
 }
 Drops_02001 {
  bEnabled=false
  emitterSpawnScript [null() {enabled=false, module=PreserveSkydiveUserVariables}]
  emitterUpdateScript [
   &EmitterState(Life Cycle Mode=::NewEnumerator1, Loop Behavior=::NewEnumerator1, Loop Duration=1.5, UseLoopDelay=false)
   &SpawnRate(SpawnRate=30, Use Spawn Probability=false)
  ]
  particleSpawnScript [
   /SetVariables(
    Particles.PositionRandom:NiagaraFloat=&UniformRangedFloat(Minimum=0, Maximum=1)
    Particles.SizeRandom:NiagaraFloat=&UniformRangedFloat(Minimum=0.13, Maximum=0.1)
    Particles.SizeRandomXZ:NiagaraFloat=&UniformRangedFloat(Minimum=0.2, Maximum=0.4)
   ) {isSetParameters=true}
   &InitializeParticle(
    Lifetime=&UniformRangedFloat(Minimum=0.4, Maximum=0.6)
    Module.Write Color=true
    Module.Write Lifetime=true
    Module.Write Mass=false
    Module.Write Material Random=true
    Module.Write Position=false
    Module.Write Scale=false
    Module.Write SpriteRotation=false
    Module.Write SpriteSize=true
    Position=&SimulationPosition()
    Sprite Rotation=&UniformRangedFloat(Minimum=0, Maximum=1)
    Sprite Size=&Vector2DFromFloat(Value=&UniformRangedFloat(Minimum=4, Maximum=6))
   )
   &SubUVAnimation(SubUV Animation Mode=::NewEnumerator1)
   &AddVelocity(Velocity=&Multiply_Vector(
     A=&VectorFromFloat(Value=500)
     B=&UniformRangedVector(Minimum=<-1 -1 -1>, Maximum=<1 1 1>)
    ))
   &SystemLocation(
    MaskSpawnGroup=false
    Offset=&Multiply_Vector(
     A=&Subtract_Vector(A=$User.TargetPosition, B=Linked() {linkedVariable {name=Engine.Owner.Position, type=NiagaraPosition}})
     B=&VectorFromFloat(Value=$Particles.PositionRandom)
    )
    Offset Coordinate Space=::Simulation
   )
  ]
  particleUpdateScript [
   &ParticleState()
   &Color(Scale Alpha=&FloatFromCurve(FloatCurve=DataInterface() {propertyValues {Curve {keys [<at -0.0046 1.0113>,<at 1.0011 0.002>]}}}))
   &ScaleSpriteSize(Scale Factor=&Vector2DFromFloat(Value=&FloatFromCurve(FloatCurve=DataInterface() {propertyValues {Curve {keys [<at 0 0>,<at 0.0524 1>,<at 0.85 1>,<at 1.0011 0.002>]}}})))
   &Drag(Drag=4, Use Linear Drag=true, Use Rotational Drag=false)
   &GravityForce()
   &SolveForcesAndVelocity(Clamp Velocity=false, Limit Acceleration=false)
   &SpriteRotationRate(Rotation Rate=$Particles.InitRotationRate)
   &ScaleSpriteSizeBySpeed(Max Scale Factor=<2 10>)
  ]
  rendererClasses [&NiagaraSpriteRendererProperties]
  renderers [{rendererClass=&NiagaraSpriteRendererProperties}]
 }
 Hand_Around_Lead {
  bEnabled=false
  emitterUpdateScript [
   &EmitterState(Life Cycle Mode=::NewEnumerator1, Loop Behavior=::NewEnumerator1, Loop Duration=2, UseLoopDelay=false)
   &SpawnRate(SpawnRate=10, Use Spawn Probability=false)
   &SpawnBurst_Instantaneous(Spawn Count=10, Spawn Group=1, Use Spawn Probability=false) {enabled=false}
  ]
  particleSpawnScript [
   /SetVariables(
    Particles.VectorNoiseForce:NiagaraFloat=&UniformRangedFloat(Minimum=50, Maximum=500)
    Particles.VortexAixs:Vector3f=&UniformRangedVector(Minimum=<-1 -1 -1>, Maximum=<1 1 1>)
    Particles.VortexForce:NiagaraFloat=&UniformRangedFloat(Minimum=1, Maximum=100)
   ) {isSetParameters=true}
   null() {enabled=false, module=FN_Orbit_Init_Variables_v2}
   /Niagara/Modules/Spawn/Location/SphereLocation(MaskSpawnGroup=false, Module.Override Randomness=false, Sphere Radius=2, Surface Only=false) {enabled=false}
   &InitializeParticle(
    Lifetime=&UniformRangedFloat(Minimum=1, Maximum=1.5)
    Module.Write Color=true
    Module.Write Lifetime=true
    Module.Write Mass=true
    Module.Write Material Random=true
    Module.Write Position=false
    Module.Write Scale=false
    Module.Write SpriteRotation=false
    Module.Write SpriteSize=true
    Position=&SimulationPosition()
    Sprite Size=<5 5>
   )
   &AddVelocity(Velocity=&Multiply_Vector(
     A=&UniformRangedVector(Minimum=<-1 -1 -1>, Maximum=<1 1 1>)
     B=&VectorFromFloat(Value=80)
    )) {enabled=false}
   &AddVelocityFromPoint(Module.Use Falloff Distance=false, Velocity Strength=&UniformRangedFloat(Minimum=50, Maximum=50))
  ]
  particleUpdateScript [
   &ParticleState()
   /Niagara/Modules/Update/Forces/CurlNoiseForce(Noise Frequency=1, Noise Quality / Cost=::NewEnumerator0, Noise Strength=500, Pan Noise=false) {enabled=false}
   &VortexForce(
    Vortex Axis=$Particles.VortexAixs
    Vortex Force Amount=&FloatFromCurve(
     FloatCurve=DataInterface() {propertyValues {Curve {keys [{arriveTangent=-0.1696, interpMode=RCIM_Cubic, leaveTangent=-0.1696, tangentMode=RCTM_User, time=0, value=0}, <at 0.8728 0.9917 RCIM_Cubic>]}}}
     Scale Curve=$Particles.VortexForce
    )
   ) {enabled=false}
   &Drag(Drag=20, Use Linear Drag=true, Use Rotational Drag=false) {enabled=false}
   &VectorNoiseForce(Force Amount=$Particles.VectorNoiseForce, Module.Use Random Noise Time Interval=false) {enabled=false}
   &SolveForcesAndVelocity(Clamp Velocity=false, Limit Acceleration=false)
   &SolveRotationalForcesAndVelocity()
   &GenerateLocationEvent(Event Send Rate=20, Module.Use Event Send Rate=true)
  ]
  rendererClasses [&NiagaraSpriteRendererProperties]
  renderers [{rendererClass=&NiagaraSpriteRendererProperties}]
 }
 Hand_Around_Ribbon {
  bEnabled=false
  emitterUpdateScript [
   &EmitterState(Life Cycle Mode=::NewEnumerator1, Loop Duration=1, UseLoopDelay=false)
   &SpawnRate(SpawnRate=5, Use Spawn Probability=false)
  ]
  particleSpawnScript [
   &InitializeParticle(
    Color=<rgba 1 0 0.3618 1>
    Module.Write Color=true
    Module.Write Lifetime=true
    Module.Write Mass=true
    Module.Write Material Random=true
    Module.Write Position=false
    Module.Write Scale=false
    Module.Write SpriteRotation=false
    Module.Write SpriteSize=true
    Position=&SimulationPosition()
    Sprite Size=<2 2>
   )
   &AddVelocity(Velocity=&Multiply_Vector(
     A=&UniformRangedVector(Minimum=<0 0 0>, Maximum=<1 1 1>)
     B=&VectorFromFloat(Value=100)
    )) {enabled=false}
   &InitializeRibbon(
    Module.Write Color=true
    Module.Write Lifetime=true
    Module.Write Mass=true
    Module.Write Material Random=true
    Module.Write Position=false
    Module.Write RibbonTwist=false
    Module.Write RibbonWidth=true
    Ribbon Width=&UniformRangedFloat(Minimum=2, Maximum=6)
   )
  ]
  particleUpdateScript [
   &ParticleState()
   &SolveForcesAndVelocity(Clamp Velocity=false, Limit Acceleration=false)
   &ScaleRibbonWidth(Ribbon Width Scale=&FloatFromCurve(FloatCurve=DataInterface() {propertyValues {Curve {keys [<at 0 1 RCIM_Cubic>,<at 1 0 RCIM_Cubic>]}}}))
  ]
  rendererClasses [&NiagaraSpriteRendererProperties, &NiagaraRibbonRendererProperties]
  renderers [
   {rendererClass=&NiagaraSpriteRendererProperties}
   {rendererClass=&NiagaraRibbonRendererProperties, rendererIndex=1}
  ]
 }
 Mesh_Body {
  bEnabled=false
  emitterSpawnScript [null() {enabled=false, module=PreserveUserVariables}]
  emitterUpdateScript [
   &EmitterState(Life Cycle Mode=::NewEnumerator1, Loop Behavior=::NewEnumerator1, Loop Duration=2, Scalability Mode=::NewEnumerator1, UseLoopDelay=false)
   &SpawnRate(SpawnRate=3, Use Spawn Probability=false)
  ]
  particleSpawnScript [
   /SetVariables(Particles.SizeRandom:NiagaraFloat=&UniformRangedFloat(Minimum=0.05, Maximum=0.15)) {isSetParameters=true, module=SetVariables_#2}
   &InitializeParticle(
    Lifetime=&UniformRangedFloat(Minimum=0.5, Maximum=1.8)
    Mesh Scale=&MakeVector(X=0.13, Y=$Particles.SizeRandom, Z=0.13)
    Module.Write Color=true
    Module.Write Lifetime=true
    Module.Write Mass=true
    Module.Write Material Random=true
    Module.Write Position=false
    Module.Write Scale=true
    Module.Write SpriteRotation=false
    Module.Write SpriteSize=false
    Position=&SimulationPosition()
   )
   /Niagara/Modules/Spawn/Orientation/V2/InitialMeshOrientation(Rotation=<0 0 0.25>, Use Orientation Vector=true, Use Rotation Vector=true)
   &InitialMeshOrientation(
    Rotation=&UniformRangedVector(Minimum=<-0.2 -0.2 0>, Maximum=<0.2 0.2 0>)
    Use Orientation Vector=true
    Use Rotation Vector=true
   ) {module=InitialMeshOrientation001}
  ]
  particleUpdateScript [
   &ParticleState()
   &DynamicMaterialParameters(
    Index 0 Param 1=$Particles.Initial.MaterialRandom
    Index 0 Param 2=&FloatFromCurve(FloatCurve=DataInterface() {propertyValues {Curve {keys [<at 0 1>,<at 1 -1>]}}})
    Module.Param0WriteEnabled=true
    Module.Param1WriteEnabled=true
    Module.Param2WriteEnabled=true
    Module.Param3WriteEnabled=true
   )
   &ScaleMeshSize(Scale Factor=&MakeVector(
     X=&FloatFromCurve(
      FloatCurve=DataInterface() {propertyValues {Curve {keys [
          <at 0 0.3815 RCIM_Cubic>
          {arriveTangent=0.0243, interpMode=RCIM_Cubic, leaveTangent=0.0243, tangentMode=RCTM_User, time=0.4967, value=0.9946}
          <at 1 0 RCIM_Cubic>
         ]}}}
      Scale Curve=&Multiply_Float(A=&Lerp_Float(A=1, B=0.2, Alpha=$Particles.SizeRandom), B=2)
     )
     Y=1
     Z=&FloatFromCurve(
      FloatCurve=DataInterface() {propertyValues {Curve {keys [
          <at 0 0.5816 RCIM_Cubic>
          {arriveTangent=-0.5816, interpMode=RCIM_Cubic, leaveTangent=-0.5816, time=0.4975, value=1.0118}
          <at 1 0 RCIM_Cubic>
         ]}}}
      Scale Curve=&Multiply_Float(A=&Lerp_Float(A=1, B=0.2, Alpha=$Particles.SizeRandom), B=2)
     )
    ))
  ]
  rendererClasses [&NiagaraMeshRendererProperties]
  renderers [{rendererClass=&NiagaraMeshRendererProperties, rendererIndex=0}]
 }
 Mesh_Body001 {
  bEnabled=false
  emitterSpawnScript [null() {enabled=false, module=PreserveUserVariables}]
  emitterUpdateScript [
   &EmitterState(Life Cycle Mode=::NewEnumerator1, Loop Behavior=::NewEnumerator1, Loop Duration=1, UseLoopDelay=false)
   &SpawnRate(SpawnRate=20, Use Spawn Probability=false)
  ]
  particleSpawnScript [
   /SetVariables(
    Particles.PositionRandom:NiagaraFloat=&UniformRangedFloat(Minimum=0, Maximum=1)
    Particles.SizeRandom:NiagaraFloat=&UniformRangedFloat(Minimum=0.13, Maximum=0.1)
    Particles.SizeRandomXZ:NiagaraFloat=&UniformRangedFloat(Minimum=0.2, Maximum=0.4)
   ) {isSetParameters=true, module=SetVariables_#2}
   &InitializeParticle(
    Lifetime=&UniformRangedFloat(Minimum=0.5, Maximum=0.8)
    Mesh Scale=&MakeVector(X=$Particles.SizeRandomXZ, Y=$Particles.SizeRandom, Z=$Particles.SizeRandomXZ)
    Module.Write Color=true
    Module.Write Lifetime=true
    Module.Write Mass=true
    Module.Write Material Random=true
    Module.Write Position=false
    Module.Write Scale=true
    Module.Write SpriteRotation=false
    Module.Write SpriteSize=false
    Position=&SimulationPosition()
   )
   &InitialMeshOrientation(
    Rotation=&UniformRangedVector(Minimum=<0 0 0>, Maximum=<1 1 1>)
    Use Orientation Vector=true
    Use Rotation Vector=true
   )
   &InitialMeshOrientation(
    Rotation=&UniformRangedVector(Minimum=<-0.05 -0.05 0>, Maximum=<0.05 0.05 0>)
    Use Orientation Vector=true
    Use Rotation Vector=true
   ) {enabled=false, module=InitialMeshOrientation001}
   &SystemLocation(
    MaskSpawnGroup=false
    Offset=&Multiply_Vector(
     A=&Subtract_Vector(A=$User.TargetPosition, B=Linked() {linkedVariable {name=Engine.Owner.Position, type=NiagaraPosition}})
     B=&VectorFromFloat(Value=$Particles.PositionRandom)
    )
    Offset Coordinate Space=::Simulation
   )
  ]
  particleUpdateScript [
   &ParticleState()
   &DynamicMaterialParameters(
    Index 0 Param 1=$Particles.Initial.MaterialRandom
    Index 0 Param 2=&FloatFromCurve(FloatCurve=DataInterface() {propertyValues {Curve {keys [<at 0 1>,<at 1 -1>]}}})
    Module.Param0WriteEnabled=true
    Module.Param1WriteEnabled=true
    Module.Param2WriteEnabled=true
    Module.Param3WriteEnabled=true
   )
   &ScaleMeshSize(Scale Factor=&MakeVector(
     X=&FloatFromCurve(
      FloatCurve=DataInterface() {propertyValues {Curve {keys [
          <at 0 0.3815 RCIM_Cubic>
          {arriveTangent=0.0243, interpMode=RCIM_Cubic, leaveTangent=0.0243, tangentMode=RCTM_User, time=0.4967, value=0.9946}
          <at 1 0 RCIM_Cubic>
         ]}}}
      Scale Curve=&Multiply_Float(A=&Lerp_Float(A=1, B=0.2, Alpha=$Particles.SizeRandom), B=2)
     )
     Y=1
     Z=&FloatFromCurve(
      FloatCurve=DataInterface() {propertyValues {Curve {keys [
          <at 0 0.5816 RCIM_Cubic>
          {arriveTangent=-0.5816, interpMode=RCIM_Cubic, leaveTangent=-0.5816, time=0.4975, value=1.0118}
          <at 1 0 RCIM_Cubic>
         ]}}}
      Scale Curve=&Multiply_Float(A=&Lerp_Float(A=1, B=0.2, Alpha=$Particles.SizeRandom), B=2)
     )
    ))
  ]
  rendererClasses [&NiagaraMeshRendererProperties]
  renderers [{rendererClass=&NiagaraMeshRendererProperties, rendererIndex=0}]
 }
 Mesh_Liquid {
  bEnabled=false
  emitterUpdateScript [
   &EmitterState(Life Cycle Mode=::NewEnumerator1, Loop Duration=1, UseLoopDelay=false)
   &SpawnRate(Use Spawn Probability=false)
  ]
  particleSpawnScript [
   &InitializeParticle(
    Lifetime=&UniformRangedFloat(Minimum=0.5, Maximum=0.8)
    Mesh Scale=&Multiply_VectorByFloat(
     Vector=&UniformRangedVector(Minimum=<0.3 0.8 1.1>, Maximum=<0.5 1.1 0.9>)
     Float=&UniformRangedFloat(Minimum=0.4, Maximum=0.2)
    )
    Module.Write Color=true
    Module.Write Lifetime=true
    Module.Write Mass=true
    Module.Write Material Random=true
    Module.Write Position=false
    Module.Write Scale=true
    Module.Write SpriteRotation=true
    Module.Write SpriteSize=true
    Position=&SimulationPosition()
   )
   &InitialMeshOrientation(
    Rotation=&UniformRangedVector(Minimum=<0 0 0>, Maximum=<1 0 0>)
    Use Orientation Vector=true
    Use Rotation Vector=true
   )
   &AddVelocityInCone(
    Cone Angle=&UniformRangedFloat(Minimum=180, Maximum=-180)
    Cone Axis=<-1 0 0>
    Module.Use Velocity Falloff On Cone Axis=false
    Velocity Distribution Along Cone Axis=2
    Velocity Strength=3
   )
   &SphereLocation(MaskSpawnGroup=false, Module.Override Randomness=false, Sphere Radius=10, Surface Only=false)
  ]
  particleUpdateScript [
   &ParticleState()
   &SolveForcesAndVelocity(Clamp Velocity=false, Limit Acceleration=false)
   &Material(Base Texture Time:NiagaraFloat=&Multiply_Float(A=$Emitter.Age, B=-1))
   &Color(
    Color=<rgba 1 1 1 1.5>
    Scale Alpha=&FloatFromCurve(FloatCurve=DataInterface() {propertyValues {Curve {keys [<at 0 0>,<at 0.05 1>,<at 0.9 1>,<at 1 0>]}}})
   ) {module=Color001}
   &DynamicMaterialParameters(
    Index 0 Param 1=$Particles.Initial.MaterialRandom
    Index 0 Param 2=&FloatFromCurve(FloatCurve=DataInterface() {propertyValues {Curve {keys [
         {interpMode=RCIM_Cubic, leaveTangent=-0.1237, tangentMode=RCTM_Break, time=0.4, value=1}
         <at 0.8685 -0.3366>
         {arriveTangent=0.0337, interpMode=RCIM_Cubic, leaveTangent=0.0337, tangentMode=RCTM_User, time=1.7, value=-1}
        ]}}})
    Index 0 Param 3=&FloatFromCurve(FloatCurve=DataInterface() {propertyValues {Curve {keys [
         {interpMode=RCIM_Cubic, tangentMode=RCTM_Break, time=0, value=0}
         {arriveTangent=100.468, interpMode=RCIM_Cubic, tangentMode=RCTM_Break, time=1, value=20}
        ]}}})
    Module.Param0WriteEnabled=true
    Module.Param1WriteEnabled=true
    Module.Param2WriteEnabled=true
    Module.Param3WriteEnabled=true
   )
   &ScaleMeshSize(Scale Factor=&Multiply_VectorByFloat(
     Vector=<1 1 1>
     Float=&FloatFromCurve(FloatCurve=DataInterface() {propertyValues {Curve {keys [
          {arriveTangent=1.9468, interpMode=RCIM_Cubic, leaveTangent=1.9468, tangentMode=RCTM_User, time=0, value=0}
          {arriveTangent=1.258, interpMode=RCIM_Cubic, leaveTangent=1.258, tangentMode=RCTM_User, time=0.3892, value=0.709}
          <at 1 1>
         ]}}})
    ))
   &ScaleMeshSizeBySpeed(Max Scale Factor=<0.5 0.8 0.8>, Min Scale Factor:Vector3f=<1.5 1 1>)
   &MeshRotationForce(Rotation=&UniformRangedVector(Minimum=<0 -5 -5>, Maximum=<0 5 5>))
   &SolveRotationalForcesAndVelocity()
  ]
  rendererClasses [&NiagaraMeshRendererProperties]
  renderers [{rendererClass=&NiagaraMeshRendererProperties, rendererIndex=0}]
 }
 Mesh_Liquid001 {
  bEnabled=false
  emitterUpdateScript [
   &EmitterState(Life Cycle Mode=::NewEnumerator1, Loop Duration=0, UseLoopDelay=false)
   &SpawnRate(SpawnRate=20, Use Spawn Probability=false)
  ]
  particleSpawnScript [
   /SetVariables(
    Particles.PositionRandom:NiagaraFloat=&UniformRangedFloat(Minimum=0, Maximum=1)
    Particles.SizeRandom:NiagaraFloat=&UniformRangedFloat(Minimum=0.13, Maximum=0.1)
    Particles.SizeRandomXZ:NiagaraFloat=&UniformRangedFloat(Minimum=0.2, Maximum=0.4)
   ) {isSetParameters=true}
   &InitializeParticle(
    Lifetime=&UniformRangedFloat(Minimum=0.5, Maximum=0.8)
    Mesh Scale=&Multiply_VectorByFloat(
     Vector=&UniformRangedVector(Minimum=<0.5 1 2>, Maximum=<1 2 1.5>)
     Float=&UniformRangedFloat(Minimum=0.4, Maximum=0.2)
    )
    Module.Write Color=true
    Module.Write Lifetime=true
    Module.Write Mass=true
    Module.Write Material Random=true
    Module.Write Position=false
    Module.Write Scale=true
    Module.Write SpriteRotation=true
    Module.Write SpriteSize=true
    Position=&SimulationPosition()
   )
   &InitialMeshOrientation(
    Rotation=&UniformRangedVector(Minimum=<0 0 0>, Maximum=<1 0 0>)
    Use Orientation Vector=true
    Use Rotation Vector=true
   )
   &AddVelocityInCone(
    Cone Angle=&UniformRangedFloat(Minimum=180, Maximum=-180)
    Cone Axis=<-1 0 0>
    Module.Use Velocity Falloff On Cone Axis=false
    Velocity Distribution Along Cone Axis=2
    Velocity Strength=3
   ) {enabled=false}
   &SphereLocation(MaskSpawnGroup=false, Module.Override Randomness=false, Sphere Radius=10, Surface Only=false) {enabled=false}
   &SystemLocation(
    MaskSpawnGroup=false
    Offset=&Multiply_Vector(
     A=&Subtract_Vector(A=$User.TargetPosition, B=Linked() {linkedVariable {name=Engine.Owner.Position, type=NiagaraPosition}})
     B=&VectorFromFloat(Value=$Particles.PositionRandom)
    )
    Offset Coordinate Space=::World
   )
   &AddVelocity(Velocity=&Multiply_Vector(
     A=&UniformRangedVector(Minimum=<-1 -1 -1>, Maximum=<1 1 1>)
     B=&VectorFromFloat(Value=20)
    ))
  ]
  particleUpdateScript [
   &ParticleState()
   &SolveForcesAndVelocity(Clamp Velocity=false, Limit Acceleration=false)
   &Material(Base Texture Time:NiagaraFloat=&Multiply_Float(A=$Emitter.Age, B=-1))
   &Color(
    Color=<rgba 1 1 1 1.5>
    Scale Alpha=&FloatFromCurve(FloatCurve=DataInterface() {propertyValues {Curve {keys [<at 0 0>,<at 0.05 1>,<at 0.9 1>,<at 1 0>]}}})
   ) {module=Color001}
   &DynamicMaterialParameters(
    Index 0 Param 1=$Particles.Initial.MaterialRandom
    Index 0 Param 2=&FloatFromCurve(FloatCurve=DataInterface() {propertyValues {Curve {keys [
         {interpMode=RCIM_Cubic, leaveTangent=-0.1237, tangentMode=RCTM_Break, time=0.4, value=1}
         <at 0.8685 -0.3366>
         {arriveTangent=0.0337, interpMode=RCIM_Cubic, leaveTangent=0.0337, tangentMode=RCTM_User, time=1.7, value=-1}
        ]}}})
    Index 0 Param 3=&FloatFromCurve(FloatCurve=DataInterface() {propertyValues {Curve {keys [
         {interpMode=RCIM_Cubic, tangentMode=RCTM_Break, time=0, value=0}
         {arriveTangent=100.468, interpMode=RCIM_Cubic, tangentMode=RCTM_Break, time=1, value=20}
        ]}}})
    Module.Param0WriteEnabled=true
    Module.Param1WriteEnabled=true
    Module.Param2WriteEnabled=true
    Module.Param3WriteEnabled=true
   )
   &ScaleMeshSize(Scale Factor=&Multiply_VectorByFloat(
     Vector=<1 1 1>
     Float=&FloatFromCurve(FloatCurve=DataInterface() {propertyValues {Curve {keys [
          {arriveTangent=1.9468, interpMode=RCIM_Cubic, leaveTangent=1.9468, tangentMode=RCTM_User, time=0, value=0}
          {arriveTangent=1.258, interpMode=RCIM_Cubic, leaveTangent=1.258, tangentMode=RCTM_User, time=0.3892, value=0.709}
          <at 1 1>
         ]}}})
    ))
   &ScaleMeshSizeBySpeed(Max Scale Factor=<0.5 0.8 0.8>, Min Scale Factor:Vector3f=<1.5 1 1>)
   &MeshRotationForce(Rotation=&UniformRangedVector(Minimum=<0 -5 -5>, Maximum=<0 5 5>))
   &SolveRotationalForcesAndVelocity()
  ]
  rendererClasses [&NiagaraMeshRendererProperties]
  renderers [{rendererClass=&NiagaraMeshRendererProperties, rendererIndex=0}]
 }
 SimpleSpriteBurst {
  bEnabled=false
  emitterUpdateScript [
   &EmitterState(Life Cycle Mode=::NewEnumerator1, Loop Behavior=::NewEnumerator1, Loop Duration=2, UseLoopDelay=false)
   &SpawnBurst_Instantaneous(Spawn Count=5, Use Spawn Probability=false)
   &SpawnRate(SpawnRate=10, Use Spawn Probability=false) {enabled=false}
  ]
  particleSpawnScript [
   /SetVariables(
    Particles.DropRandom:NiagaraFloat=&UniformAOrBFloat(A=0)
    Particles.PendulumDrag=5
    Particles.PendulumLengthScale:NiagaraFloat=&Multiply_Float(A=0.125, B=&UniformRangedFloat(Minimum=0.5, Maximum=1))
    Particles.PendulumMaskFalloff=1.5
    Particles.PendulumRigidity=0.2
    Particles.PivotOffset:Vector3f=&UniformRangedVector(Minimum=<-4 -4 -4>, Maximum=<4 4 4>)
   ) {isSetParameters=true}
   &InitializeParticle(
    Color=<rgba 0 0 0 1>
    Lifetime=&UniformRangedFloat(Minimum=2, Maximum=3)
    Mesh Scale=&Multiply_Vector(A=<1 1 1>, B=&MakeVector(X=0.1, Y=0.1, Z=$Particles.PendulumLengthScale))
    Module.Write Color=true
    Module.Write Lifetime=true
    Module.Write Mass=true
    Module.Write Material Random=true
    Module.Write Position=false
    Module.Write Scale=true
    Module.Write SpriteRotation=false
    Module.Write SpriteSize=true
    Sprite Size=<1 1>
   )
   /SetVariables(Particles.RandomOffset:Vector3f=/Niagara/DynamicInputs/Random/RandomVector(
     Vector Scale:NiagaraFloat=&UniformRangedFloat(Minimum=25, Maximum=202.999)
     Constraint Vector:Vector3f=<0 0 1>
     Randomness Mode:ENiagaraRandomnessMode=::NewEnumerator0
     Random Seed:NiagaraInt32=0
     Fixed Random Seed:NiagaraBool=false
    )) {enabled=false, isSetParameters=true}
  ]
  particleUpdateScript [
   &SkeletalMeshLocation(
    Apply To Particle Position=false
    Bone / Socket Index=1
    Bone Sampling Mode=::NewEnumerator4
    Inherit Velocity=false
    MaskSpawnGroup=false
    Sampled Position Offset Coordinate Space:ENiagaraMeshTransforms=::NewEnumerator3
    Skeletal Mesh=$User.SkeletalMesh
   )
   &SystemLocation(MaskSpawnGroup=false, Offset=$Particles.PivotOffset) {enabled=false}
   &ParticleState()
   /SetVariables(Particles.MeshOrigin:Vector3f=&Add_Vector(
     A=$Particles.PivotOffset
     B=&Add_Vector(
      A=$Output.SkeletalMeshLocation.SampledPosition
      B=&MakeVector(
       X=0
       Y=0
       Z=&FloatFromCurve(
        FloatCurve=DataInterface() {propertyValues {Curve {keys [
            <at 0 0>
            {arriveTangent=-0.0697, interpMode=RCIM_Cubic, leaveTangent=-0.0697, tangentMode=RCTM_User, time=0.7, value=0}
            {arriveTangent=5.5154, interpMode=RCIM_Cubic, leaveTangent=5.4075, tangentMode=RCTM_User, time=1, value=1}
           ]}}}
        Scale Curve=&Multiply_Float(A=-170, B=$Particles.DropRandom)
       )
      )
     )
    )) {isSetParameters=true}
   &Drag(Drag=$Particles.PendulumDrag, Use Linear Drag=true, Use Rotational Drag=false)
   &GravityForce(Gravity=<0 0 -900>)
   &PendulumSetup(
    Module.Enable Spring Driver=false
    Pendulum Length:NiagaraFloat=&Multiply_Float(A=100, B=$Particles.PendulumLengthScale)
    Pendulum Pivot:Vector3f=$Particles.MeshOrigin
    "Spring Driven Constraint (Tightness)"=3
   )
   &SolveForcesAndVelocity(Clamp Velocity=false, Limit Acceleration=false)
   &PendulumConstraint()
   &CalculateAccurateVelocity()
   /SetVariables(Particles.MeshFacing:Vector3f=&NormalizeVector(Vector To Normalize=&Subtract_Vector(A=$Particles.MeshOrigin, B=Linked() {linkedVariable {name=Particles.Position, type=NiagaraPosition}}))) {isSetParameters=true}
   /Niagara/Modules/Update/Orientation/V2/OrientMeshToVector(
    Look At Direction:Vector3f=&Multiply_VectorByFloat(Vector=$Particles.MeshFacing, Float=1)
    Override Quat=false
    Reference Axis Vector:Vector3f=<0 0 1>
    UseOptionalSideAxis=false
    UseOptionalUpAxis=false
   )
   &DynamicMaterialParameters(
    Index 2 Param 1=HlslExpression() {hlslExpression=Particles.MeshOrigin.x}
    Index 2 Param 2=HlslExpression() {hlslExpression=Particles.MeshOrigin.y}
    Index 2 Param 3=HlslExpression() {hlslExpression=Particles.MeshOrigin.z}
    Index 2 Param 4=$Particles.PendulumMaskFalloff
    Index 3 Param 1=&MakeFloatFromVector(Vector=Linked() {linkedVariable {name=Particles.Position, type=NiagaraPosition}}, Channel=::NewEnumerator0)
    Index 3 Param 2=&MakeFloatFromVector(Vector=Linked() {linkedVariable {name=Particles.Position, type=NiagaraPosition}}, Channel=::NewEnumerator1)
    Index 3 Param 3=&MakeFloatFromVector(Vector=Linked() {linkedVariable {name=Particles.Position, type=NiagaraPosition}}, Channel=::NewEnumerator2)
    Index 3 Param 4=$Particles.PendulumRigidity
    Module.Param0WriteEnabled2=true
    Module.Param0WriteEnabled3=true
    Module.Param1WriteEnabled2=true
    Module.Param1WriteEnabled3=true
    Module.Param2WriteEnabled2=true
    Module.Param2WriteEnabled3=true
    Module.Param3WriteEnabled2=true
    Module.Param3WriteEnabled3=true
    Write Parameter Index 0=false
    Write Parameter Index 2=true
    Write Parameter Index 3=true
   )
   &ScaleMeshSize(Scale Factor=&MakeVector(
     X=&FloatFromCurve(FloatCurve=DataInterface() {propertyValues {Curve {keys [<at 0 0.2 RCIM_Cubic>,<at 0.7 1 RCIM_Cubic>,<at 1 0.2>]}}})
     Y=&FloatFromCurve(FloatCurve=DataInterface() {propertyValues {Curve {keys [<at 0 0.2 RCIM_Cubic>,<at 0.7 1 RCIM_Cubic>,<at 1 0.2 RCIM_Cubic>]}}})
     Z=&FloatFromCurve(FloatCurve=DataInterface() {propertyValues {Curve {keys [
          <at 0 0.25 RCIM_Cubic>
          <at 0.7 0.8055 RCIM_Cubic>
          {arriveTangent=-1.5491, interpMode=RCIM_Cubic, leaveTangent=-1.5491, tangentMode=RCTM_User, time=1, value=0.25}
         ]}}})
    ))
   &CameraOffset(Camera Offset Amount=2)
  ]
  rendererClasses [&NiagaraMeshRendererProperties]
  renderers [{rendererClass=&NiagaraMeshRendererProperties, rendererIndex=0}]
 }
 TendrilNet {
  bEnabled=false
  emitterUpdateScript [
   &EmitterState(Life Cycle Mode=::NewEnumerator1, Loop Behavior=::NewEnumerator1, Loop Duration=4, UseLoopDelay=false)
   &SpawnRate(SpawnRate=5, Use Spawn Probability=false)
   &SpawnBurst_Instantaneous(Spawn Count=5, Use Spawn Probability=false)
  ]
  particleSpawnScript [
   /SetVariables(
    Particles.D_TimeOffset:NiagaraFloat=&UniformRangedFloat(Minimum=0, Maximum=5)
    Particles.Random=&UniformRangedFloat(Minimum=1, Maximum=5)
    Particles.UV_Scale:NiagaraFloat=&UniformRangedFloat(Minimum=0.1, Maximum=0.5)
   ) {isSetParameters=true, module=SetVariables_#1}
   &InitializeParticle(
    Color=<rgba 0.0053 0.0068 0.0156 1>
    Lifetime=&UniformRangedFloat(Minimum=0.5, Maximum=1.8)
    Module.Write Color=true
    Module.Write Lifetime=true
    Module.Write Mass=true
    Module.Write Material Random=true
    Module.Write Position=false
    Module.Write Scale=false
    Module.Write SpriteRotation=true
    Module.Write SpriteSize=true
    Position=&SimulationPosition()
    Sprite Rotation=&UniformRangedFloat(Minimum=0, Maximum=1)
    Sprite Size=&Multiply_Vector2(
     A=&UniformRangedVector2D(Minimum=<0.5 0.5>, Maximum=<1 1>)
     B=&Vector2DFromFloat(Value=&UniformRangedFloat(Minimum=15, Maximum=25))
    )
   )
   &AddVelocityFromPoint(Module.Use Falloff Distance=false, Velocity Strength=&UniformRangedFloat(Minimum=10, Maximum=15))
   &SphereLocation(MaskSpawnGroup=false, Module.Override Randomness=false, Sphere Radius=25, Surface Only=false)
   &InheritVelocity(
    Module.Limit Speed=true
    Velocity Scale=&Multiply_Vector(A=<1 1 1>, B=&VectorFromFloat(Value=$Particles.Random))
   ) {enabled=false}
  ]
  particleUpdateScript [
   &ParticleState()
   &PointForce(Force Strength=-60, Module.Use Falloff=false, Module.UseLinearForce=false)
   &Drag(Drag=3, Use Linear Drag=true, Use Rotational Drag=false)
   &SolveForcesAndVelocity(Clamp Velocity=false, Limit Acceleration=false)
   &DynamicMaterialParameters(
    Index 0 Param 1=$Particles.D_TimeOffset
    Index 0 Param 2=$Particles.UV_Scale
    Index 0 Param 3=0.6
    Index 0 Param 4=0.4
    Module.Param0WriteEnabled=true
    Module.Param1WriteEnabled=true
    Module.Param2WriteEnabled=true
    Module.Param3WriteEnabled=true
   )
   &ScaleColor(
    Scale Alpha=&FloatFromCurve(FloatCurve=DataInterface() {propertyValues {Curve {keys [<at 0 0 RCIM_Cubic>,<at 0.502 2.9766 RCIM_Cubic>,<at 1 0>]}}})
    ScaleA=true
    ScaleRGB=true
   )
   &ScaleSpriteSize(Scale Factor=&Vector2DFromFloat(Value=&FloatFromCurve(FloatCurve=DataInterface() {propertyValues {Curve {keys [<at 0.64 1 RCIM_Cubic>,<at 1.64 0 RCIM_Cubic>]}}})))
  ]
  rendererClasses [&NiagaraSpriteRendererProperties]
  renderers [{rendererClass=&NiagaraSpriteRendererProperties}]
 }
 Tendrils {
  bEnabled=false
  emitterUpdateScript [
   &EmitterState(Life Cycle Mode=::NewEnumerator1, Loop Behavior=::NewEnumerator1, Loop Duration=2, UseLoopDelay=false)
   &SpawnRate(SpawnRate=2, Use Spawn Probability=false)
   &SpawnBurst_Instantaneous(Spawn Count=8, Use Spawn Probability=false)
  ]
  particleSpawnScript [
   /SetVariables(
    Particles.BendDirectionX:NiagaraFloat=&UniformRangedFloat(Minimum=0, Maximum=1)
    Particles.BendDirectionY:NiagaraFloat=&UniformRangedFloat(Minimum=0, Maximum=1)
    Particles.BendDirectionZ:NiagaraFloat=&UniformRangedFloat(Minimum=0, Maximum=1)
    Particles.BendStrength:NiagaraFloat=&UniformRangedFloat(Minimum=60, Maximum=150)
    Particles.CurveOffset:NiagaraFloat=&UniformRangedFloat(Minimum=0, Maximum=1)
    Particles.PanSpeed:NiagaraFloat=&UniformRangedFloat(Minimum=-1, Maximum=-2)
    Particles.SinScale:NiagaraFloat=&UniformRangedFloat(Minimum=0.1, Maximum=0.2)
    Particles.SinSpeed:NiagaraFloat=&Lerp_Float(A=-8, B=-15, Alpha=$Particles.SinScale)
    Particles.WidthRandom:NiagaraFloat=&UniformRangedFloat(Minimum=0.3, Maximum=1.7)
   ) {isSetParameters=true}
   &InitializeParticle(
    Color=<rgba 0.0052 0.0052 0.0052 1>
    Lifetime=&UniformRangedFloat(Minimum=0.7, Maximum=1.7)
    Mesh Scale=&Multiply_Vector(
     A=<1 0.6 0.6>
     B=&VectorFromFloat(Value=&UniformRangedFloat(Minimum=0.6, Maximum=0.8))
    )
    Module.Write Color=true
    Module.Write Lifetime=true
    Module.Write Mass=true
    Module.Write Material Random=true
    Module.Write Position=false
    Module.Write Scale=true
    Module.Write SpriteRotation=false
    Module.Write SpriteSize=true
    Position=&SimulationPosition()
    Sprite Size=<5 5>
   )
   &InitialMeshOrientation(
    Rotation=&UniformRangedVector(Minimum=<-1 -1 -1>, Maximum=<1 1 1>)
    Use Orientation Vector=true
    Use Rotation Vector=true
   )
   &AddVelocityFromPoint(Module.Use Falloff Distance=false, Velocity Strength=1)
  ]
  particleUpdateScript [
   &ParticleState()
   &SolveForcesAndVelocity(Clamp Velocity=false, Limit Acceleration=false)
   &DynamicMaterialParameters(
    Index 0 Param 1=&FloatFromCurve(
     FloatCurve=DataInterface() {propertyValues {Curve {keys [
         <at 0 0.2 RCIM_Cubic>
         {arriveTangent=2.0368, interpMode=RCIM_Cubic, leaveTangent=2.0368, time=0.0982, value=2}
         {arriveTangent=-0.6576, interpMode=RCIM_Cubic, leaveTangent=-0.6576, time=0.2706, value=0.7511}
         <at 0.4018 1.8004 RCIM_Cubic>
         {arriveTangent=-1.2864, interpMode=RCIM_Cubic, leaveTangent=-1.2864, time=0.5508, value=0.7511}
         <at 0.7243 1.3855>
         <at 1 0 RCIM_Cubic>
        ]}}}
     Scale Curve=$Particles.WidthRandom
    )
    Index 0 Param 3=&FloatFromCurve(FloatCurve=DataInterface() {propertyValues {Curve {keys [
         <at 0 -0.5 RCIM_Cubic>
         {arriveTangent=2.0829, interpMode=RCIM_Cubic, leaveTangent=2.0829, time=0.3991, value=0.9166}
         <at 0.5434 1.0587>
         {arriveTangent=-2.3574, interpMode=RCIM_Cubic, leaveTangent=-2.3574, time=0.6718, value=0.8994}
         <at 1 -0.5 RCIM_Cubic>
        ]}}})
    Index 0 Param 4=$Particles.PanSpeed
    Index 1 Param 1=&FloatFromCurve(
     FloatCurve=DataInterface() {propertyValues {Curve {
        postInfinityExtrap=RCCE_CycleWithOffset
        preInfinityExtrap=RCCE_CycleWithOffset
        keys [
         <at -0.2078 -0.2949 RCIM_Cubic>
         {arriveTangent=-0.1876, interpMode=RCIM_Cubic, leaveTangent=-0.1876, time=-0.1305, value=0.3531}
         {arriveTangent=0.0791, interpMode=RCIM_Cubic, leaveTangent=0.0791, time=-0.0639, value=-0.3219}
         {arriveTangent=-0.0573, interpMode=RCIM_Cubic, leaveTangent=-0.0573, time=0.0458, value=0.3671}
         {arriveTangent=0.1517, interpMode=RCIM_Cubic, leaveTangent=0.1517, time=0.1269, value=-0.3329}
         {arriveTangent=-0.1648, interpMode=RCIM_Cubic, leaveTangent=-0.1648, time=0.2208, value=0.3937}
         {arriveTangent=0.1554, interpMode=RCIM_Cubic, leaveTangent=0.1554, time=0.3191, value=-0.3646}
         {arriveTangent=0.1069, interpMode=RCIM_Cubic, leaveTangent=0.1069, time=0.4235, value=0.4252}
         {arriveTangent=-0.3187, interpMode=RCIM_Cubic, leaveTangent=-0.3187, time=0.5167, value=-0.3434}
         {arriveTangent=0.1258, interpMode=RCIM_Cubic, leaveTangent=0.1258, time=0.6494, value=0.3531}
         {arriveTangent=0.1369, interpMode=RCIM_Cubic, leaveTangent=0.1369, time=0.7949, value=-0.3084}
         {arriveTangent=-0.1066, interpMode=RCIM_Cubic, leaveTangent=-0.1066, time=0.9125, value=0.3892}
         <at 1.0905 -0.3399 RCIM_Cubic>
         {arriveTangent=0.0092, interpMode=RCIM_Cubic, leaveTangent=0.0092, time=1.2112, value=0.3892}
         {arriveTangent=0.0148, interpMode=RCIM_Cubic, leaveTangent=0.0148, time=1.3361, value=-0.3377}
         {arriveTangent=0.0109, interpMode=RCIM_Cubic, leaveTangent=0.0109, time=1.4646, value=0.3929}
         {arriveTangent=-0.8851, interpMode=RCIM_Cubic, leaveTangent=-0.8851, time=1.603, value=-0.3348}
         {arriveTangent=0.5332, interpMode=RCIM_Cubic, leaveTangent=0.5332, time=1.6678, value=0.2131}
         {arriveTangent=0.6123, interpMode=RCIM_Cubic, leaveTangent=0.6123, time=1.7796, value=-0.2406}
         {arriveTangent=-0.2327, interpMode=RCIM_Cubic, leaveTangent=-0.2327, time=1.8915, value=0.3501}
         <at 2.0004 -0.292 RCIM_Cubic>
        ]
       }}}
     CurveIndex=&Add_Float(A=$Particles.NormalizedAge, B=$Particles.CurveOffset)
     Scale Curve=$Particles.BendStrength
    )
    Index 1 Param 2=$Particles.BendDirectionX
    Index 1 Param 3=$Particles.BendDirectionY
    Index 1 Param 4=$Particles.BendDirectionZ
    Index 2 Param 1=$Particles.SinScale
    Index 2 Param 2=$Particles.SinSpeed
    Index 2 Param 3=5
    Module.Param0WriteEnabled=true
    Module.Param0WriteEnabled1=true
    Module.Param0WriteEnabled2=true
    Module.Param1WriteEnabled=true
    Module.Param1WriteEnabled1=true
    Module.Param1WriteEnabled2=true
    Module.Param2WriteEnabled=true
    Module.Param2WriteEnabled1=true
    Module.Param2WriteEnabled2=true
    Module.Param3WriteEnabled=true
    Module.Param3WriteEnabled1=true
    Module.Param3WriteEnabled2=true
    Write Parameter Index 1=true
    Write Parameter Index 2=true
   )
   &ScaleMeshSize(Scale Factor=&VectorFromFloat(Value=&FloatFromCurve(FloatCurve=DataInterface() {propertyValues {Curve {keys [<at 0.5214 1 RCIM_Cubic>,<at 1 0 RCIM_Cubic>]}}})))
  ]
  rendererClasses [&NiagaraSpriteRendererProperties, &NiagaraMeshRendererProperties]
  renderers [
   {rendererClass=&NiagaraSpriteRendererProperties}
   {rendererClass=&NiagaraMeshRendererProperties, rendererIndex=1}
  ]
 }
 TendrilsShell {
  bEnabled=false
  emitterUpdateScript [
   &EmitterState(Life Cycle Mode=::NewEnumerator1, Loop Behavior=::NewEnumerator1, Loop Duration=3, UseLoopDelay=false)
   &SpawnRate(SpawnRate=20, Use Spawn Probability=false) {enabled=false}
   &SpawnBurst_Instantaneous(Spawn Count=1, Use Spawn Probability=false)
  ]
  particleSpawnScript [
   &InitializeParticle(
    Color=<rgba 1 1 1 10>
    Lifetime=2.7
    Mesh Scale=&Multiply_Vector(A=<0.8 1 1>, B=&VectorFromFloat(Value=0.6))
    Module.Write Color=true
    Module.Write Lifetime=true
    Module.Write Mass=true
    Module.Write Material Random=true
    Module.Write Position=false
    Module.Write Scale=true
    Module.Write SpriteRotation=false
    Module.Write SpriteSize=true
    Position=&SimulationPosition()
    Sprite Size=<5 5>
   )
   &InitialMeshOrientation(Rotation=<0 0 0.5>, Use Orientation Vector=true, Use Rotation Vector=true)
   &SystemLocation(MaskSpawnGroup=false, Offset=<2 0 0>)
  ]
  particleUpdateScript [
   &ParticleState()
   &SolveForcesAndVelocity(Clamp Velocity=false, Limit Acceleration=false)
   &DynamicMaterialParameters(
    Index 0 Param 2=&FloatFromCurve(FloatCurve=DataInterface() {propertyValues {Curve {keys [
         <at 0 -1 RCIM_Cubic>
         {arriveTangent=-0.03, interpMode=RCIM_Cubic, leaveTangent=-0.03, tangentMode=RCTM_User, time=0.1833, value=0.12}
         {arriveTangent=-3.4406, interpMode=RCIM_Cubic, leaveTangent=-3.4406, tangentMode=RCTM_User, time=1, value=-1}
        ]}}})
    Index 0 Param 3=4
    Module.Param0WriteEnabled=true
    Module.Param1WriteEnabled=true
    Module.Param2WriteEnabled=true
    Module.Param3WriteEnabled=true
   )
   &CameraOffset(Camera Offset Amount=5) {enabled=false}
   &ScaleColor(
    Scale Alpha=&FloatFromCurve(FloatCurve=DataInterface() {propertyValues {Curve {keys [
         <at 0 0 RCIM_Cubic>
         {arriveTangent=0.1008, interpMode=RCIM_Cubic, leaveTangent=0.1008, tangentMode=RCTM_User, time=0.0523, value=1}
         {arriveTangent=0.275, interpMode=RCIM_Cubic, leaveTangent=0.275, tangentMode=RCTM_User, time=0.6836, value=1}
         <at 1 0 RCIM_Cubic>
        ]}}})
    ScaleA=true
    ScaleRGB=true
   )
   &ScaleMeshSize(Scale Factor=&MakeVector(
     X=&FloatFromCurve(FloatCurve=DataInterface() {propertyValues {Curve {keys [<at 0.0039 1.5 RCIM_Cubic>,<at 0.1737 2>,<at 1 1 RCIM_Cubic>]}}})
     Y=1
     Z=1
    ))
   &CameraOffset(Camera Offset Amount=&FloatFromCurve(FloatCurve=DataInterface() {propertyValues {Curve {keys [<at 0 1 RCIM_Cubic>,<at 1 0>]}}}, Scale Curve=20)) {module=CameraOffset001}
  ]
  rendererClasses [&NiagaraSpriteRendererProperties, &NiagaraMeshRendererProperties]
  renderers [
   {rendererClass=&NiagaraSpriteRendererProperties}
   {rendererClass=&NiagaraMeshRendererProperties, rendererIndex=1}
  ]
 }
 Tendrils_New {
  bEnabled=false
  emitterUpdateScript [
   &EmitterState(Life Cycle Mode=::NewEnumerator1, Loop Behavior=::NewEnumerator1, Loop Duration=3, UseLoopDelay=false)
   &SpawnRate(SpawnRate=20, Use Spawn Probability=false) {enabled=false}
   &SpawnBurst_Instantaneous(Spawn Count=1, Use Spawn Probability=false)
  ]
  particleSpawnScript [
   &InitializeParticle(
    Color=<rgba 1 1 1 10>
    Lifetime=2
    Module.Write Color=true
    Module.Write Lifetime=true
    Module.Write Mass=true
    Module.Write Material Random=true
    Module.Write Position=false
    Module.Write Scale=true
    Module.Write SpriteRotation=false
    Module.Write SpriteSize=true
    Position=&SimulationPosition()
    Sprite Size=<5 5>
   )
   &InitialMeshOrientation(Rotation=<0 0 0.5>, Use Orientation Vector=true, Use Rotation Vector=true)
   &SystemLocation(MaskSpawnGroup=false)
  ]
  particleUpdateScript [
   &ParticleState()
   &SolveForcesAndVelocity(Clamp Velocity=false, Limit Acceleration=false)
   &DynamicMaterialParameters(
    Index 0 Param 1=&FloatFromCurve(FloatCurve=DataInterface() {propertyValues {Curve {keys [
         <at 0 -3>
         <at 0.0629 0>
         {arriveTangent=-3.2013, interpMode=RCIM_Cubic, leaveTangent=-3.2013, time=0.6471, value=0}
         {interpMode=RCIM_Cubic, leaveTangent=-21.8502, time=1, value=-3}
        ]}}})
    Index 0 Param 2=&FloatFromCurve(
     FloatCurve=DataInterface() {propertyValues {Curve {keys [
         <at 0 0 RCIM_Cubic>
         {arriveTangent=0.2217, interpMode=RCIM_Cubic, leaveTangent=0.2217, time=0.1109, value=0.3}
         <at 0.8935 0.1981>
         {arriveTangent=-3.4406, interpMode=RCIM_Cubic, leaveTangent=-3.4406, tangentMode=RCTM_User, time=1, value=0}
        ]}}}
     Scale Curve=2
    )
    Index 0 Param 3=&FloatFromCurve(
     FloatCurve=DataInterface() {propertyValues {Curve {keys [
         <at 0 -1 RCIM_Cubic>
         {arriveTangent=1.341, interpMode=RCIM_Cubic, leaveTangent=1.341, time=0.1109, value=0.3}
         {arriveTangent=-1.4622, interpMode=RCIM_Cubic, leaveTangent=-1.4622, tangentMode=RCTM_User, time=0.8935, value=0.1981}
         {arriveTangent=-3.4406, interpMode=RCIM_Cubic, leaveTangent=-3.4406, tangentMode=RCTM_User, time=1, value=0}
        ]}}}
     Scale Curve=6
    )
    Module.Param0WriteEnabled=false
    Module.Param0WriteEnabled1=true
    Module.Param1WriteEnabled=true
    Module.Param1WriteEnabled1=true
    Module.Param2WriteEnabled=true
    Module.Param2WriteEnabled1=true
    Module.Param3WriteEnabled=true
    Module.Param3WriteEnabled1=true
    Write Parameter Index 1=true
   )
   &CameraOffset(Camera Offset Amount=5) {enabled=false}
   &ScaleColor(
    Scale Alpha=&FloatFromCurve(FloatCurve=DataInterface() {propertyValues {Curve {keys [
         <at 0 0 RCIM_Cubic>
         {arriveTangent=0.1008, interpMode=RCIM_Cubic, leaveTangent=0.1008, tangentMode=RCTM_User, time=0.06, value=1.9203}
         {arriveTangent=0.275, interpMode=RCIM_Cubic, leaveTangent=0.275, tangentMode=RCTM_User, time=0.7114, value=1.7081}
         <at 1 0 RCIM_Cubic>
        ]}}})
    ScaleA=true
    ScaleRGB=true
   )
   &ScaleMeshSize(Scale Factor=&MakeVector(
     X=&FloatFromCurve(FloatCurve=DataInterface() {propertyValues {Curve {keys [<at 0.0039 1.5 RCIM_Cubic>,<at 0.1737 2>,<at 1 1 RCIM_Cubic>]}}})
     Y=1
     Z=1
    )) {enabled=false}
   &CameraOffset(Camera Offset Amount=&FloatFromCurve(FloatCurve=DataInterface() {propertyValues {Curve {keys [<at 0 1 RCIM_Cubic>,<at 1 0>]}}}, Scale Curve=20)) {module=CameraOffset001}
   /Niagara/Modules/Update/Orientation/V2/UpdateMeshOrientation(Rotation Rate=-0.3, Rotation Vector:Vector3f=<1 0 0>)
  ]
  rendererClasses [&NiagaraSpriteRendererProperties, &NiagaraMeshRendererProperties]
  renderers [
   {rendererClass=&NiagaraSpriteRendererProperties}
   {rendererClass=&NiagaraMeshRendererProperties, rendererIndex=1}
  ]
 }
}
summary {
 systemName=NS_HighTower_Vertigo_Hand_FX
 emitters [
  {bEnabled=false, emitterName=Drops_02, rendererClasses [&NiagaraSpriteRendererProperties]}
  {bEnabled=false, emitterName=Mesh_Liquid, rendererClasses [&NiagaraMeshRendererProperties]}
  {bEnabled=false, emitterName=Mesh_Body, rendererClasses [&NiagaraMeshRendererProperties]}
  {bEnabled=false, emitterName=Mesh_Body001, rendererClasses [&NiagaraMeshRendererProperties]}
  {bEnabled=false, emitterName=Drops_02001, rendererClasses [&NiagaraSpriteRendererProperties]}
  {bEnabled=false, emitterName=Mesh_Liquid001, rendererClasses [&NiagaraMeshRendererProperties]}
  {bEnabled=false, emitterName=Hand_Around_Lead, rendererClasses [&NiagaraSpriteRendererProperties]}
  {
   bEnabled=false
   emitterName=Hand_Around_Ribbon
   rendererClasses [&NiagaraSpriteRendererProperties, &NiagaraRibbonRendererProperties]
  }
  {bEnabled=false, emitterName=ArmGuid, rendererClasses [&NiagaraSpriteRendererProperties]}
  {
   bEnabled=false
   emitterName=ArmStrike
   rendererClasses [&NiagaraSpriteRendererProperties, &NiagaraRibbonRendererProperties]
  }
  {bEnabled=false, emitterName=TendrilNet, rendererClasses [&NiagaraSpriteRendererProperties]}
  {
   bEnabled=false
   emitterName=Arm_Net
   rendererClasses [&NiagaraMeshRendererProperties, &NiagaraSpriteRendererProperties]
  }
  {
   bEnabled=false
   emitterName=Tendrils
   rendererClasses [&NiagaraSpriteRendererProperties, &NiagaraMeshRendererProperties]
  }
  {
   bEnabled=false
   emitterName=TendrilsShell
   rendererClasses [&NiagaraSpriteRendererProperties, &NiagaraMeshRendererProperties]
  }
  {bEnabled=false, emitterName=SimpleSpriteBurst, rendererClasses [&NiagaraMeshRendererProperties]}
  {
   bEnabled=false
   emitterName=Tendrils_New
   rendererClasses [&NiagaraSpriteRendererProperties, &NiagaraMeshRendererProperties]
  }
 ]
 userVariables [
  User.bIsFrontEndPreview {type=NiagaraBool}
  User.FEVec {type=Vector3f, description="", defaultValue {struct=Vector3f, value=<0 0 0>}}
  User.PlayerSpeed {type=NiagaraFloat, defaultValue=0, description=""}
  User.PreviousVelocityXY {type=Vector3f, description="", defaultValue {struct=Vector3f, value=<0 0 0>}}
  User.RibbonIDOffset {type=NiagaraInt32}
  User.TargetPosition {type=Vector3f, description="", defaultValue {struct=Vector3f, value=<0 0 0>}}
  User.WaterBodyIndex {type=NiagaraFloat, defaultValue=0, description=""}
  User.WaterDepth {type=NiagaraFloat, defaultValue=0, description=""}
  User.WaterPlaneLocation {type=Vector3f, description="", defaultValue {struct=Vector3f, value=<0 0 0>}}
  User.WaterPlaneNormal {type=Vector3f, description="", defaultValue {struct=Vector3f, value=<0 0 0>}}
  User.WaveTime {type=NiagaraFloat, defaultValue=0, description=""}
 ]
}
