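# Niagara impact effect for the HighTower Vertigo item, with three one-shot burst emitters, all disabled in this asset.
# GlowDrops throws speed-stretched sprites under gravity that fade from an HDR flash, thinning spawn count by distance.
# Splash spawns short-lived streak sprites driven by random and age-based material parameters.
# Tendrils adds dark, randomly oriented meshes and sprites, feeding curve-driven width, bend, pan and sine wobble into dynamic material parameters.
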NiagaraSystem /HighTower/Items/Vertigo/FX/NS_HighTower_Vertigo_Impact
dependencies {
 usedDataInterfaces [&NiagaraDataInterfaceCurve, &NiagaraDataInterfaceVector2DCurve, &NiagaraDataInterfaceColorCurve]
 usedDynamicInputs []
 usedModules [/Niagara/Modules/System/SystemLifeCycle, &SystemState]
}
emitters {
 GlowDrops {
  bEnabled=false
  emitterUpdateScript [
   &EmitterState(
    Enable Distance Culling=true
    Life Cycle Mode=::NewEnumerator1
    Loop Behavior=::NewEnumerator1
    Loop Duration=1
    Max Distance Response=::KillImmediately
    Module.bApplySpawnCountScaleByDistance=true
    Module.bUseMaxDistance=true
    Module.bUseMinDistance=false
    Scalability Mode=::NewEnumerator1
    Scale Spawn Count=true
    Spawn Count Scale By Distance=DataInterface() {propertyValues {Curve {keys [<at 0.3676 0.9657>,<at 0.9962 0.4982>]}}}
    UseLoopDelay=false
   )
   &SpawnBurst_Instantaneous(Spawn Count=&UniformRangedInt(Minimum=8, Maximum=12), Use Spawn Probability=false)
  ]
  particleSpawnScript [
   /Niagara/Modules/Spawn/Initialization/InitializeParticle(
    Color=<rgba 0.08 0.08 0.08 1>
    Lifetime=&UniformRangedFloat(Minimum=0.3, Maximum=0.5)
    Module.Write Color=true
    Module.Write Lifetime=true
    Module.Write Mass=true
    Module.Write Material Random=true
    Module.Write Position=false
    Module.Write Scale=false
    Module.Write SpriteRotation=false
    Module.Write SpriteSize=true
    Position=&SimulationPosition()
    Sprite Size=&Multiply_Vector2(
     A=&UniformRangedVector2D(Minimum=<0.5 1>, Maximum=<1 2>)
     B=&Vector2DFromFloat(Value=&UniformRangedFloat(Minimum=10, Maximum=20))
    )
   )
   /Niagara/Modules/Spawn/Location/SphereLocation(MaskSpawnGroup=false, Module.Override Randomness=false, Sphere Radius=5, Surface Only=false)
   &AddVelocityFromPoint(
    Module.Use Falloff Distance=false
    Origin Offset=<-1 0 0>
    Velocity Strength=&UniformRangedFloat(Minimum=300, Maximum=500)
   )
  ]
  particleUpdateScript [
   &ParticleState()
   &GravityForce(Gravity=<0 0 -500>)
   &SolveForcesAndVelocity(Clamp Velocity=false, Limit Acceleration=false)
   &ScaleSpriteSize(Scale Factor=&Vector2DFromCurve(
     Vector2Curve=DataInterface() {propertyValues {
       XCurve {keys [
         {arriveTangent=5.9728, interpMode=RCIM_Cubic, leaveTangent=5.9728, tangentMode=RCTM_User, time=0, value=0.1}
         {arriveTangent=-0.1, interpMode=RCIM_Cubic, leaveTangent=-0.1, tangentMode=RCTM_User, time=0.18, value=0.99}
         {interpMode=RCIM_Cubic, tangentMode=RCTM_User, time=1, value=0}
        ]}
       YCurve {keys [
         {arriveTangent=6.2715, interpMode=RCIM_Cubic, leaveTangent=6.2715, tangentMode=RCTM_User, time=0, value=0.1}
         {arriveTangent=-0.1001, interpMode=RCIM_Cubic, leaveTangent=-0.1001, tangentMode=RCTM_User, time=0.18, value=0.99}
         {interpMode=RCIM_Cubic, tangentMode=RCTM_User, time=0.9988, value=0}
        ]}
      }}
     ScaleCurve=<1 1>
    ))
   &ScaleSpriteSizeBySpeed(Max Scale Factor=<0.6 8>)
   &ScaleColor(
    Scale Alpha=&FloatFromCurve(FloatCurve=DataInterface() {propertyValues {Curve {keys [<at 0.4 1.0113 RCIM_Cubic>,<at 1.0011 0.002>]}}})
    Scale RGB=&MakeVectorFromLinearColorRGB(Color=&ColorFromCurve(ColorCurve=DataInterface() {propertyValues {
        AlphaCurve {keys []}
        BlueCurve {keys [<at 0 10 RCIM_Cubic>,<at 0.65 2>]}
        GreenCurve {keys [<at 0 10 RCIM_Cubic>,<at 0.65 2>]}
        RedCurve {keys [<at 0 10 RCIM_Cubic>,<at 0.65 2>]}
       }}))
    ScaleA=true
    ScaleRGB=true
   )
  ]
  rendererClasses [&NiagaraSpriteRendererProperties]
  renderers [{rendererClass=&NiagaraSpriteRendererProperties}]
 }
 Splash {
  bEnabled=false
  emitterUpdateScript [
   &EmitterState(Life Cycle Mode=::NewEnumerator1, Loop Behavior=::NewEnumerator1, Loop Duration=1, UseLoopDelay=false)
   &SpawnRate(Use Spawn Probability=false) {enabled=false}
   &SpawnBurst_Instantaneous(Spawn Count=5, Use Spawn Probability=false)
  ]
  particleSpawnScript [
   /SetVariables(Particles.Random=&UniformRangedFloat(Minimum=0, Maximum=1)) {isSetParameters=true}
   &InitializeParticle(
    Lifetime=&UniformRangedFloat(Minimum=0.3, Maximum=0.4)
    Module.Write Color=true
    Module.Write Lifetime=true
    Module.Write Mass=true
    Module.Write Material Random=true
    Module.Write Position=false
    Module.Write Scale=false
    Module.Write SpriteRotation=false
    Module.Write SpriteSize=true
    Position=&SimulationPosition()
    Sprite Size=&Multiply_Vector2(
     A=<20 2>
     B=&Vector2DFromFloat(Value=&UniformRangedFloat(Minimum=10, Maximum=15))
    )
   )
   &AddVelocityFromPoint(Module.Use Falloff Distance=false, Velocity Strength=0.1)
  ]
  particleUpdateScript [
   &ParticleState()
   &SolveForcesAndVelocity(Clamp Velocity=false, Limit Acceleration=false)
   &DynamicMaterialParameters(
    Index 0 Param 1=$Particles.Random
    Index 0 Param 2=&FloatFromCurve(FloatCurve=DataInterface() {propertyValues {Curve {keys [<at 0 1 RCIM_Cubic>,<at 1 0 RCIM_Cubic>]}}})
    Module.Param0WriteEnabled=true
    Module.Param1WriteEnabled=true
    Module.Param2WriteEnabled=true
    Module.Param3WriteEnabled=true
   )
   &ScaleSpriteSize(Scale Factor=&MakeVector2D(
     X=&FloatFromCurve(FloatCurve=DataInterface() {propertyValues {Curve {keys [
          <at 0 0.1 RCIM_Cubic>
          {arriveTangent=0.0472, interpMode=RCIM_Cubic, leaveTangent=0.0472, tangentMode=RCTM_User, time=0.2544, value=1}
          {arriveTangent=-3.2781, interpMode=RCIM_Cubic, leaveTangent=-3.2781, tangentMode=RCTM_User, time=1, value=0}
         ]}}})
     Y=1
    ))
  ]
  rendererClasses [&NiagaraSpriteRendererProperties]
  renderers [{rendererClass=&NiagaraSpriteRendererProperties}]
 }
 Tendrils {
  bEnabled=false
  emitterUpdateScript [
   &EmitterState(Life Cycle Mode=::NewEnumerator1, Loop Behavior=::NewEnumerator1, Loop Duration=1, UseLoopDelay=false)
   &SpawnRate(SpawnRate=8, Use Spawn Probability=false) {enabled=false}
   &SpawnBurst_Instantaneous(Spawn Count=&UniformRangedInt(Minimum=5, Maximum=8), Use Spawn Probability=false)
  ]
  particleSpawnScript [
   /SetVariables(
    Particles.BendDirectionX:NiagaraFloat=&UniformRangedFloat(Minimum=0, Maximum=1)
    Particles.BendDirectionY:NiagaraFloat=&UniformRangedFloat(Minimum=0, Maximum=1)
    Particles.BendDirectionZ:NiagaraFloat=&UniformRangedFloat(Minimum=0, Maximum=1)
    Particles.BendStrength:NiagaraFloat=&UniformRangedFloat(Minimum=30, Maximum=50)
    Particles.CurveOffset:NiagaraFloat=&UniformRangedFloat(Minimum=0, Maximum=1)
    Particles.PanSpeed:NiagaraFloat=&UniformRangedFloat(Minimum=-1, Maximum=-2.5)
    Particles.SinScale:NiagaraFloat=&UniformRangedFloat(Minimum=0.1, Maximum=0.2)
    Particles.SinSpeed:NiagaraFloat=&Lerp_Float(A=-6, B=-8, Alpha=$Particles.SinScale)
    Particles.WidthRandom:NiagaraFloat=&UniformRangedFloat(Minimum=1, Maximum=2)
   ) {isSetParameters=true}
   &InitializeParticle(
    Color=<rgba 0.0052 0.0052 0.0052 1>
    Lifetime=&UniformRangedFloat(Minimum=0.2, Maximum=0.4)
    Mesh Scale=&Multiply_Vector(
     A=<1 1 1>
     B=&VectorFromFloat(Value=&UniformRangedFloat(Minimum=1, Maximum=2))
    )
    Module.Write Color=true
    Module.Write Lifetime=true
    Module.Write Mass=true
    Module.Write Material Random=true
    Module.Write Position=false
    Module.Write Scale=true
    Module.Write SpriteRotation=false
    Module.Write SpriteSize=true
    Position=&SimulationPosition()
    Sprite Size=<5 5>
   )
   /Niagara/Modules/Spawn/Orientation/V2/InitialMeshOrientation(
    Rotation=&UniformRangedVector(Minimum=<-1 -1 -1>, Maximum=<1 1 1>)
    Use Orientation Vector=true
    Use Rotation Vector=true
   )
   &AddVelocityFromPoint(Module.Use Falloff Distance=false, Velocity Strength=1)
  ]
  particleUpdateScript [
   &ParticleState()
   &SolveForcesAndVelocity(Clamp Velocity=false, Limit Acceleration=false)
   &DynamicMaterialParameters(
    Index 0 Param 1=&FloatFromCurve(
     FloatCurve=DataInterface() {propertyValues {Curve {keys [
         <at 0 0 RCIM_Cubic>
         {arriveTangent=2.7759, interpMode=RCIM_Cubic, leaveTangent=2.7759, time=0.0982, value=2}
         {arriveTangent=-0.6576, interpMode=RCIM_Cubic, leaveTangent=-0.6576, time=0.2706, value=0.7511}
         <at 0.4018 1.8004 RCIM_Cubic>
         {arriveTangent=-1.2864, interpMode=RCIM_Cubic, leaveTangent=-1.2864, time=0.5508, value=0.7511}
         <at 0.7243 1.3855>
         <at 1 0 RCIM_Cubic>
        ]}}}
     Scale Curve=$Particles.WidthRandom
    )
    Index 0 Param 3=&FloatFromCurve(FloatCurve=DataInterface() {propertyValues {Curve {keys [
         <at 0 -0.5 RCIM_Cubic>
         {arriveTangent=2.0829, interpMode=RCIM_Cubic, leaveTangent=2.0829, time=0.3991, value=0.9166}
         <at 0.5434 1.0587>
         {arriveTangent=-2.3574, interpMode=RCIM_Cubic, leaveTangent=-2.3574, time=0.6718, value=0.8994}
         <at 1 -0.5 RCIM_Cubic>
        ]}}})
    Index 0 Param 4=$Particles.PanSpeed
    Index 1 Param 1=&FloatFromCurve(
     FloatCurve=DataInterface() {propertyValues {Curve {
        postInfinityExtrap=RCCE_CycleWithOffset
        preInfinityExtrap=RCCE_CycleWithOffset
        keys [
         <at -0.2078 -0.2949 RCIM_Cubic>
         {arriveTangent=-0.1876, interpMode=RCIM_Cubic, leaveTangent=-0.1876, time=-0.1305, value=0.3531}
         {arriveTangent=0.0791, interpMode=RCIM_Cubic, leaveTangent=0.0791, time=-0.0639, value=-0.3219}
         {arriveTangent=-0.0573, interpMode=RCIM_Cubic, leaveTangent=-0.0573, time=0.0458, value=0.3671}
         {arriveTangent=0.1517, interpMode=RCIM_Cubic, leaveTangent=0.1517, time=0.1269, value=-0.3329}
         {arriveTangent=-0.1648, interpMode=RCIM_Cubic, leaveTangent=-0.1648, time=0.2208, value=0.3937}
         {arriveTangent=0.1554, interpMode=RCIM_Cubic, leaveTangent=0.1554, time=0.3191, value=-0.3646}
         {arriveTangent=0.1069, interpMode=RCIM_Cubic, leaveTangent=0.1069, time=0.4235, value=0.4252}
         {arriveTangent=-0.3187, interpMode=RCIM_Cubic, leaveTangent=-0.3187, time=0.5167, value=-0.3434}
         {arriveTangent=0.1258, interpMode=RCIM_Cubic, leaveTangent=0.1258, time=0.6494, value=0.3531}
         {arriveTangent=0.1369, interpMode=RCIM_Cubic, leaveTangent=0.1369, time=0.7949, value=-0.3084}
         {arriveTangent=-0.1066, interpMode=RCIM_Cubic, leaveTangent=-0.1066, time=0.9125, value=0.3892}
         <at 1.0905 -0.3399 RCIM_Cubic>
         {arriveTangent=0.0092, interpMode=RCIM_Cubic, leaveTangent=0.0092, time=1.2112, value=0.3892}
         {arriveTangent=0.0148, interpMode=RCIM_Cubic, leaveTangent=0.0148, time=1.3361, value=-0.3377}
         {arriveTangent=0.0109, interpMode=RCIM_Cubic, leaveTangent=0.0109, time=1.4646, value=0.3929}
         {arriveTangent=-0.8851, interpMode=RCIM_Cubic, leaveTangent=-0.8851, time=1.603, value=-0.3348}
         {arriveTangent=0.5332, interpMode=RCIM_Cubic, leaveTangent=0.5332, time=1.6678, value=0.2131}
         {arriveTangent=0.6123, interpMode=RCIM_Cubic, leaveTangent=0.6123, time=1.7796, value=-0.2406}
         {arriveTangent=-0.2327, interpMode=RCIM_Cubic, leaveTangent=-0.2327, time=1.8915, value=0.3501}
         <at 2.0004 -0.292 RCIM_Cubic>
        ]
       }}}
     CurveIndex=&Add_Float(A=$Particles.NormalizedAge, B=$Particles.CurveOffset)
     Scale Curve=$Particles.BendStrength
    )
    Index 1 Param 2=$Particles.BendDirectionX
    Index 1 Param 3=$Particles.BendDirectionY
    Index 1 Param 4=$Particles.BendDirectionZ
    Index 2 Param 1=$Particles.SinScale
    Index 2 Param 2=$Particles.SinSpeed
    Index 2 Param 3=10
    Module.Param0WriteEnabled=true
    Module.Param0WriteEnabled1=true
    Module.Param0WriteEnabled2=true
    Module.Param1WriteEnabled=true
    Module.Param1WriteEnabled1=true
    Module.Param1WriteEnabled2=true
    Module.Param2WriteEnabled=true
    Module.Param2WriteEnabled1=true
    Module.Param2WriteEnabled2=true
    Module.Param3WriteEnabled=true
    Module.Param3WriteEnabled1=true
    Module.Param3WriteEnabled2=true
    Write Parameter Index 1=true
    Write Parameter Index 2=true
   )
   &ScaleMeshSize(Scale Factor=&VectorFromFloat(Value=&FloatFromCurve(FloatCurve=DataInterface() {propertyValues {Curve {keys [<at 0.5214 1 RCIM_Cubic>,<at 1 0.5 RCIM_Cubic>]}}})))
  ]
  rendererClasses [&NiagaraSpriteRendererProperties, &NiagaraMeshRendererProperties]
  renderers [
   {rendererClass=&NiagaraSpriteRendererProperties}
   {rendererClass=&NiagaraMeshRendererProperties, rendererIndex=1}
  ]
 }
}
summary {
 systemName=NS_HighTower_Vertigo_Impact
 emitters [
  {
   bEnabled=false
   emitterName=Tendrils
   rendererClasses [&NiagaraSpriteRendererProperties, &NiagaraMeshRendererProperties]
  }
  {bEnabled=false, emitterName=GlowDrops, rendererClasses [&NiagaraSpriteRendererProperties]}
  {bEnabled=false, emitterName=Splash, rendererClasses [&NiagaraSpriteRendererProperties]}
 ]
 userVariables [
  User.bIsFrontEndPreview {type=NiagaraBool}
  User.FEVec {type=Vector3f, description="", defaultValue {struct=Vector3f, value=<0 0 0>}}
  User.PlayerSpeed {type=NiagaraFloat, defaultValue=0, description=""}
  User.PreviousVelocityXY {type=Vector3f, description="", defaultValue {struct=Vector3f, value=<0 0 0>}}
  User.RibbonIDOffset {type=NiagaraInt32}
  User.TargetPosition {type=Vector3f, description="", defaultValue {struct=Vector3f, value=<0 0 0>}}
  User.WaterBodyIndex {type=NiagaraFloat, defaultValue=0, description=""}
  User.WaterDepth {type=NiagaraFloat, defaultValue=0, description=""}
  User.WaterPlaneLocation {type=Vector3f, description="", defaultValue {struct=Vector3f, value=<0 0 0>}}
  User.WaterPlaneNormal {type=Vector3f, description="", defaultValue {struct=Vector3f, value=<0 0 0>}}
  User.WaveTime {type=NiagaraFloat, defaultValue=0, description=""}
 ]
}
