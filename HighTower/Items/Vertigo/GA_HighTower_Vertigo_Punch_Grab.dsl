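; Vertigo punch grab ability, built on a knockback parent, that yanks the hit target toward its launcher.
; AI pawns are launched on a suggested arc to a point near the launcher; players and player bots use a curved root-motion move-to-actor force.
; A gameplay event tells the launcher to switch montage section, sent on launch for pawns or ability end for vehicles.
; The player path also reports a quest stat when not locally controlled.
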(event Custom|LaunchAI
  (params (Target :type Fort%20Pawn%20Object%20Reference))
  (bind (_start _end) (CallFunction|CalculateStartandEndPos))
  ;[0.9973 0.9534 1] LaunchAI-> Pull via launch for Henchmen and Sharks
  (bind _outlaunchvelocity (Game|SuggestProjectileVelocityCustomArc _start _end 0.0 (Variables|Default|GetArcParam))) ; Finds out how much force is needed to put the target 300 units in front of instigator
  ;[1 1 1] Using "launch character jump" wouldnt always fire for some reason
  (Character|LaunchCharacter Target _outlaunchvelocity true true)
  (CallFunction|LaunchFinished))

(event Ability|EventOnEndAbility
  (params (bWasCancelled :type Boolean))
  ;[1 1 1] Fires if target is a vehicle
  (Utilities|Casting|CastToFortVehicleInterface (Variables|KnockbackParent|GetTarget))
  (Ability|SendGameplayEventtoActor (Variables|KnockbackParent|GetLauncher) "(TagName=\"Event.HighTower.Vertigo.Punch.Proj.Hit\")" (Utilities|Struct|MakeGameplayEventData :ContextHandle "()" :InstigatorTags "(GameplayTags=)" :TargetTags "(GameplayTags=)" :EventMagnitude (Math|Conversions|ToFloat(Boolean) (Variables|Default|GetHitValidTarget)) :TargetData "()")) ; Sends event to change montage section to yank ability
  (|CallParentFunction bWasCancelled))

(event EventDoLaunch
  (params (Vel :type Vector :default "0, 0, 0"))
  ;[1 1 1] Fires if target is a pawn
  (Development|PrintString "Launch Ability a-go-go" false false)
  (Ability|SendGameplayEventtoActor (Variables|KnockbackParent|GetLauncher) "(TagName=\"Event.HighTower.Vertigo.Punch.Proj.Hit\")" (Utilities|Struct|MakeGameplayEventData :ContextHandle "()" :InstigatorTags "(GameplayTags=)" :TargetTags "(GameplayTags=)" :EventMagnitude (Math|Conversions|ToFloat(Boolean) (Variables|Default|GetHitValidTarget)) :TargetData "()")) ; Sends event to change montage section to yank ability
  (Actor|SetLifeSpan (Ability|EffectContext|GetEffectCauser (Ability|GetGrantedbyEffectContext)) 0.2)
  (bind _asfort_pawn (Utilities|Casting|CastToFortPawn (Variables|KnockbackParent|GetTarget)))
  (bind _asfort_aipawn (Utilities|Casting|CastToFortAIPawn _asfort_pawn)
    (:then
      (CallFunction|LaunchAI _asfort_aipawn))
    (:CastFailed
      ;[1 1 1] Root motion pull for players and player bots.
      (Ability|Tasks|ApplyRootMotionMoveToActorForce "None" (Variables|KnockbackParent|GetLauncher) (Math|Vector|MakeVector (Variables|Default|GetDistanceFromInstigator_Players)) "AlignFromTargetToSource" 0.5 0 0 false "MOVE_None" false "/HighTower/Items/Vertigo/Curve_HighTower_Vertigo_Grab_Movement.Curve_HighTower_Vertigo_Grab_Movement" 0 "SetVelocity" "0, 0, 0" 0.0 true
        (:then
          ;[1 1 1] Quest hook
          (Ability|IsLocallyControlled
            (:False
              (bind _asfort_player_pawn (Utilities|Casting|CastToFortPlayerPawn (Variables|KnockbackParent|GetLauncher)))
              (Quest|SendComplexCustomStatEvent (Quests|GetQuestManager (Fortnite|GetFortPlayerControllerfromActor _asfort_player_pawn) "Invalid") 0 "(GameplayTags=((TagName=\"Athena.Quests.Ability.HighTower.Power.Vertigo.Punch\")))" "(GameplayTags=)" 1))))
        (:OnFinished
          (CallFunction|LaunchFinished))))))

(fn CalculateStartAndEndPos
  (return (Transformation|GetActorLocation (Variables|KnockbackParent|GetTarget)) (+ (Transformation|GetActorLocation (Variables|KnockbackParent|GetLauncher)) (* (Math|Vector|Normalize (- (Transformation|GetActorLocation (Variables|KnockbackParent|GetTarget)) (Transformation|GetActorLocation (Variables|KnockbackParent|GetLauncher)))) (Variables|Default|GetDistanceFromInstigator_AI)))))

(fn FindLauncher
  (params (DefaultLauncher :type Actor%20Object%20Reference))
  (return (Actor|GetOwner DefaultLauncher)))

(fn HandleMomentum
  (bind _launcher (Variables|KnockbackParent|GetLauncher))
  (Utilities|IsValid _launcher
    (:Is%20Valid
      (Variables|Default|SetMomentum (Transformation|GetVelocity _launcher)))))

(fn MomentumCheck
  (params (PreMomentumLaunchvelocity :type Vector))
  (return (select (> (Math|Vector|VectorLength (+ PreMomentumLaunchvelocity (Variables|Default|GetMomentum))) (Math|Vector|VectorLength PreMomentumLaunchvelocity)) (+ PreMomentumLaunchvelocity (Variables|Default|GetMomentum)) PreMomentumLaunchvelocity)))

(fn OnRep_Momentum)
